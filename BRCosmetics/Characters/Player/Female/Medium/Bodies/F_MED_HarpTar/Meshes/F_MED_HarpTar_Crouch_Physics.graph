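# Crouch physics asset for the HarpTar female medium body cosmetic, covering its two-sided boa and waist bag.
# Capsules on pelvis, spine, forearms and thighs form the body volumes, with forearm and thigh constraints locked in swing and twist.
# Boa segments chain from spine_04 with mostly limited swing and tight twist.
# The left boa root and the waist bag, constrained to the left thigh, rotate freely.
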PhysicsAsset /BRCosmetics/Characters/Player/Female/Medium/Bodies/F_MED_HarpTar/Meshes/F_MED_HarpTar_Crouch_Physics
bodies {
 def_boa_left_03 [{
   shapeType=Capsule
   length=5.179
   radius=4.9379
   shapeName=def_boa_left_03_capsule
   center=<4.7594 0.1561 1.5225>
   rotation=<rot 0 172.1397 -57.2215>
  }]
 def_boa_left_04 [{
   shapeType=Capsule
   length=9.5991
   radius=4.0105
   shapeName=def_boa_left_04_capsule
   center=<5.6235 -0.1297 0.2574>
   rotation=<rot -0.1884 -88.4014 -85.5107>
  }]
 def_boa_right_04 [{
   shapeType=Capsule
   length=6.9435
   radius=5.2003
   shapeName=def_boa_right_04_capsule
   center=<4.9821 0.0702 1.1612>
   rotation=<rot 0 -93.1066 -75.6571>
  }]
 dyn_boa_left_05 [{
   shapeType=Capsule
   length=6.7461
   radius=3.7591
   shapeName=dyn_boa_left_05_capsule
   center=<6.4507 1.1507 -0.3324>
   rotation=<rot -4.4696 -86.5118 -83.8023>
  }]
 dyn_boa_left_06 [{
   shapeType=Capsule
   length=11.1088
   radius=3.1327
   shapeName=dyn_boa_left_06_capsule
   center=<8.9064 1.3617 -1.3034>
   rotation=<rot 0.0161 87.3933 -89.5939>
  }]
 dyn_boa_right_05 [{
   shapeType=Capsule
   length=8.9222
   radius=3.9632
   shapeName=dyn_boa_right_05_capsule
   center=<4.7567 -1.3439 -0.3333>
   rotation=<rot 20.3619 -92.4324 -87.7251>
  }]
 dyn_boa_right_06 [{
   shapeType=Capsule
   length=12.1309
   radius=3.3342
   shapeName=dyn_boa_right_06_capsule
   center=<7.69 -1.1294 -1.5214>
   rotation=<rot -0.0088 95.6584 -89.9547>
  }]
 dyn_waist_bag [{
   shapeType=Capsule
   length=9.3734
   radius=7.2946
   shapeName=dyn_waist_bag_capsule
   center=<4.4571 0.0235 -0.1871>
   rotation=<rot 0 2.889 -88.1613>
  }]
 lowerarm_l [{
   shapeType=Capsule
   length=14.4557
   radius=2.9205
   shapeName=lowerarm_l_capsule
   center=<9.4966 -0.2103 0>
   rotation=<rot 89.0047 -0.003 -0.003>
  }]
 lowerarm_r [{
   shapeType=Capsule
   length=14.4557
   radius=2.9205
   shapeName=lowerarm_r_capsule
   center=<-9.4966 0.2103 0>
   rotation=<rot 89.0047 -0.003 179.997>
  }]
 pelvis [{shapeType=Capsule, length=7.2341, radius=12.0369, shapeName=pelvis_capsule, center=<-1.3075 -2.309 0>}]
 spine_02 [{
   shapeType=Capsule
   length=4.9516
   radius=8.2737
   shapeName=spine_02_capsule
   center=<2.6603 1.9407 -0.0014>
   rotation=<rot -0.0022 -90 0.0062>
  }]
 spine_04 [
  {shapeType=Capsule, length=5.4005, radius=9.7844, shapeName=spine_04_capsule, center=<0.8244 2.3929 0>}
  {shapeType=Capsule, length=7.58, radius=10.2364, shapeName=spine_04_capsule, center=<3.833 -0.2458 0>}
 ]
 thigh_l [{shapeType=Capsule, length=26.2574, radius=7.5306, shapeName=thigh_l_capsule, center=<-21.8671 0 0>, rotation=<rot 88.8481 179.9996 179.9997>}]
 thigh_r [{shapeType=Capsule, length=26.2574, radius=7.5306, shapeName=thigh_r_capsule, center=<21.8671 0 0>, rotation=<rot 88.8481 179.9996 -0.0003>}]
}
constraints [
 {
  bone1Name=lowerarm_l
  bone2Name=spine_04
  swing1LimitDegrees=45
  swing1Motion=Locked
  swing2LimitDegrees=45
  swing2Motion=Locked
  twistLimitDegrees=45
  twistMotion=Locked
 }
 {
  bone1Name=thigh_l
  bone2Name=pelvis
  swing1LimitDegrees=45
  swing1Motion=Locked
  swing2LimitDegrees=45
  swing2Motion=Locked
  twistLimitDegrees=45
  twistMotion=Locked
 }
 {
  bone1Name=thigh_r
  bone2Name=pelvis
  swing1LimitDegrees=45
  swing1Motion=Locked
  swing2LimitDegrees=45
  swing2Motion=Locked
  twistLimitDegrees=45
  twistMotion=Locked
 }
 {
  bone1Name=lowerarm_r
  bone2Name=spine_04
  swing1LimitDegrees=45
  swing1Motion=Locked
  swing2LimitDegrees=45
  swing2Motion=Locked
  twistLimitDegrees=45
  twistMotion=Locked
 }
 {bone1Name=dyn_boa_right_06, bone2Name=dyn_boa_right_05, swing1LimitDegrees=25, swing2LimitDegrees=25, twistLimitDegrees=2}
 {bone1Name=dyn_boa_right_05, bone2Name=def_boa_right_04, swing1LimitDegrees=25, swing2LimitDegrees=25, twistLimitDegrees=2}
 {bone1Name=dyn_boa_left_06, bone2Name=dyn_boa_left_05, swing1LimitDegrees=30, swing2LimitDegrees=30, twistLimitDegrees=2}
 {bone1Name=dyn_boa_left_05, bone2Name=def_boa_left_04, swing1LimitDegrees=30, swing2LimitDegrees=30, twistLimitDegrees=2}
 {bone1Name=def_boa_right_04, bone2Name=spine_04, swing1LimitDegrees=15, swing2LimitDegrees=15, twistLimitDegrees=15}
 {bone1Name=def_boa_left_04, bone2Name=def_boa_left_03, swing1LimitDegrees=10, swing2LimitDegrees=10, twistLimitDegrees=4}
 {
  bone1Name=dyn_waist_bag
  bone2Name=thigh_l
  swing1LimitDegrees=45
  swing1Motion=Free
  swing2LimitDegrees=45
  swing2Motion=Free
  twistLimitDegrees=45
  twistMotion=Free
 }
 {
  bone1Name=def_boa_left_03
  bone2Name=spine_04
  swing1LimitDegrees=45
  swing1Motion=Free
  swing2LimitDegrees=45
  swing2Motion=Free
  twistLimitDegrees=45
  twistMotion=Free
 }
]
