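# Control Rig for the F_MED_HarpTar female cosmetic body, procedurally correcting its boa and bag-belt bones.
# After a shared deform rig function, a branch on RBAN selects the pinning and boa solver functions.
# The boa uses weighted parent constraints on nulls and left/right spline IK.
# Spherical pose readers on arms and thighs, plus state flags like crouching, targeting, DBNO and parachute, weight additive transform offsets, some smoothed by interpolation.
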ControlRigBlueprint /BRCosmetics/Characters/Player/Female/Medium/Bodies/F_MED_HarpTar/Meshes/F_MED_HarpTar_CtrlRig
graphs [
 New Function_ContainedGraph {nodes [
   RigVMFunctionEntryNode()
   > &RigUnit_ParentConstraint(
    AdvancedSettings=(RotationOrderForFilter=XZY)
    Child=(Type=Bone,Name="boa_root")
    Parents="((Item=(Type=Bone,Name=\"spine_01\"),Weight=0.0119),(Item=(Type=Bone,Name=\"spine_02\"),Weight=0.0388),(Item=(Type=Bone,Name=\"spine_03\"),Weight=0.0433),(Item=(Type=Bone,Name=\"clavicle_l\"),Weight=0.1148),(Item=(Type=Bone,Name=\"clavicle_r\"),Weight=0.1217),(Item= ...(truncated)"
    Weight=1
   )
   > &RigUnit_ParentConstraint(
    AdvancedSettings=(RotationOrderForFilter=XZY)
    Child=(Type=Null,Name="def_boa_left_04_null")
    Parents="((Item=(Type=Bone,Name=\"spine_04\"),Weight=0.0039),(Item=(Type=Bone,Name=\"spine_05\"),Weight=0.0235),(Item=(Type=Bone,Name=\"deform_upperarm_up_l\"),Weight=0.0636),(Item=(Type=Bone,Name=\"clavicle_l\"),Weight=0.132),(Item=(Type=Bone,Name=\"deform_upperarm_fr_l\"), ...(truncated)"
    Weight=1
    bMaintainOffset=true
   )
   > &RigUnit_ParentConstraint(
    AdvancedSettings=(RotationOrderForFilter=XZY)
    Child=(Type=Null,Name="def_boa_left_03_null")
    Parents="((Item=(Type=Bone,Name=\"deform_upperarm_fr_l\"),Weight=0.0175),(Item=(Type=Bone,Name=\"deform_upperarm_bk_l\"),Weight=0.1385),(Item=(Type=Bone,Name=\"deform_upperarm_up_l\"),Weight=0.2167),(Item=(Type=Bone,Name=\"clavicle_l\"),Weight=0.2509),(Item=(Type=Bone,Name ...(truncated)"
    Weight=1
   )
   > &RigUnit_ParentConstraint(
    AdvancedSettings=(RotationOrderForFilter=XZY)
    Child=(Type=Null,Name="def_boa_left_02_null")
    Parents="((Item=(Type=Bone,Name=\"spine_04\"),Weight=0.0114),(Item=(Type=Bone,Name=\"spine_05\"),Weight=0.0247),(Item=(Type=Bone,Name=\"deform_upperarm_up_l\"),Weight=0.0841),(Item=(Type=Bone,Name=\"clavicle_l\"),Weight=0.2759),(Item=(Type=Bone,Name=\"deform_upperarm_bk_l\") ...(truncated)"
    Weight=1
   )
   > &RigUnit_ParentConstraint(
    AdvancedSettings=(RotationOrderForFilter=XZY)
    Child=(Type=Null,Name="def_boa_left_01_null")
    Parents="((Item=(Type=Bone,Name=\"spine_02\"),Weight=0.0057),(Item=(Type=Bone,Name=\"spine_03\"),Weight=0.0072),(Item=(Type=Bone,Name=\"clavicle_r\"),Weight=0.0108),(Item=(Type=Bone,Name=\"deform_upperarm_bk_l\"),Weight=0.013),(Item=(Type=Bone,Name=\"upperarm_twist_01_l\"),W ...(truncated)"
    Weight=1
   )
   > &RigUnit_ParentConstraint(
    AdvancedSettings=(RotationOrderForFilter=XZY)
    Child=(Type=Null,Name="def_boa_right_01_null")
    Parents="((Item=(Type=Bone,Name=\"spine_02\"),Weight=0.0057),(Item=(Type=Bone,Name=\"spine_03\"),Weight=0.0072),(Item=(Type=Bone,Name=\"clavicle_l\"),Weight=0.0108),(Item=(Type=Bone,Name=\"deform_upperarm_bk_r\"),Weight=0.013),(Item=(Type=Bone,Name=\"upperarm_twist_01_r\"),W ...(truncated)"
    Weight=1
   )
   > &RigUnit_ParentConstraint(
    AdvancedSettings=(RotationOrderForFilter=XZY)
    Child=(Type=Null,Name="def_boa_right_02_null")
    Parents="((Item=(Type=Bone,Name=\"spine_04\"),Weight=0.0169),(Item=(Type=Bone,Name=\"spine_05\"),Weight=0.038),(Item=(Type=Bone,Name=\"deform_upperarm_up_r\"),Weight=0.0796),(Item=(Type=Bone,Name=\"upperarm_twist_01_r\"),Weight=0.2352),(Item=(Type=Bone,Name=\"deform_upperar ...(truncated)"
    Weight=1
   )
   > &RigUnit_ParentConstraint(
    AdvancedSettings=(RotationOrderForFilter=XZY)
    Child=(Type=Null,Name="def_boa_right_03_null")
    Parents="((Item=(Type=Bone,Name=\"deform_upperarm_fr_r\"),Weight=0.0175),(Item=(Type=Bone,Name=\"deform_upperarm_bk_r\"),Weight=0.1385),(Item=(Type=Bone,Name=\"deform_upperarm_up_r\"),Weight=0.2167),(Item=(Type=Bone,Name=\"clavicle_r\"),Weight=0.2509),(Item=(Type=Bone,Name ...(truncated)"
    Weight=1
    bMaintainOffset=true
   )
   > &RigUnit_ParentConstraint(
    AdvancedSettings=(RotationOrderForFilter=XZY)
    Child=(Type=Null,Name="def_boa_right_04_null")
    Parents="((Item=(Type=Bone,Name=\"spine_04\"),Weight=0.0039),(Item=(Type=Bone,Name=\"spine_05\"),Weight=0.0235),(Item=(Type=Bone,Name=\"deform_upperarm_up_r\"),Weight=0.0636),(Item=(Type=Bone,Name=\"clavicle_r\"),Weight=0.132),(Item=(Type=Bone,Name=\"deform_upperarm_fr_r\"), ...(truncated)"
    Weight=1
   )
   > RigVMFunctionReturnNode()
  ]}
 New Function_ContainedGraph {nodes [
   RigVMFunctionEntryNode()
   > RigVMFunction_Sequence = RigVMAggregateNode()
   n1 = &RigVMFunction_Sequence(ExecuteContext=@RigVMFunction_Sequence.B)
   /Script/ControlRig.RigUnit_SetRelativeTransformForItem(
    Child=(Type=Bone,Name="def_boa_right_02")
    Parent=(Type=Bone,Name="def_boa_right_02")
    Value=(Rotation=(X=0,Y=0,Z=-0.1736,W=0.9848),Translation=(X=0,Y=0,Z=3),Scale3D=(X=1,Y=1,Z=1))
    Weight=1
    bParentInitial=False
    bPropagateToChildren=True
    ExecutePin=@n1.A
   ) {
    is=struct RigVMUnitNode
    template="Set Relative Transform::Execute(in Child,in Parent,in bParentInitial,in Value,in Weight,in bPropagateToChildren)"
   }
   > &RigUnit_SetRelativeTransformForItem(
    Child=(Type=Bone,Name="def_boa_right_03")
    Parent=(Type=Bone,Name="def_boa_right_03")
    Value=(Rotation=(X=0,Y=0.0872,Z=0,W=0.9962),Translation=(X=0,Y=0,Z=0),Scale3D=(X=1,Y=1,Z=1))
    Weight=1
    bParentInitial=False
    bPropagateToChildren=True
   ) {template="Set Relative Transform::Execute(in Child,in Parent,in bParentInitial,in Value,in Weight,in bPropagateToChildren)"}
   > &RigUnit_SetRelativeTransformForItem(
    Child=(Type=Bone,Name="def_boa_right_04")
    Parent=(Type=Bone,Name="def_boa_right_04")
    Value=(Rotation=(X=0,Y=0,Z=0.1219,W=0.9925),Translation=(X=0,Y=0,Z=0),Scale3D=(X=1,Y=1,Z=1))
    Weight=1
    bParentInitial=False
    bPropagateToChildren=True
   ) {template="Set Relative Transform::Execute(in Child,in Parent,in bParentInitial,in Value,in Weight,in bPropagateToChildren)"}
   > &RigUnit_SetRelativeTransformForItem(
    Child=(Type=Bone,Name="def_boa_left_02")
    Parent=(Type=Bone,Name="def_boa_left_02")
    Value=(Rotation=(X=-0.0872,Y=0,Z=0,W=0.9962),Translation=(X=0,Y=1,Z=1),Scale3D=(X=1,Y=1,Z=1))
    Weight=1
    bParentInitial=False
    bPropagateToChildren=True
   ) {template="Set Relative Transform::Execute(in Child,in Parent,in bParentInitial,in Value,in Weight,in bPropagateToChildren)"}
   > &RigUnit_SetRelativeTransformForItem(
    Child=(Type=Bone,Name="def_boa_left_03")
    Parent=(Type=Bone,Name="def_boa_left_03")
    Value=(Rotation=(X=0,Y=0,Z=0,W=1),Translation=(X=-1,Y=0,Z=0),Scale3D=(X=1,Y=1,Z=1))
    Weight=1
    bParentInitial=False
    bPropagateToChildren=True
   ) {template="Set Relative Transform::Execute(in Child,in Parent,in bParentInitial,in Value,in Weight,in bPropagateToChildren)"}
   > &RigUnit_SetRelativeTransformForItem(
    Child=(Type=Bone,Name="def_boa_left_04")
    Parent=(Type=Bone,Name="def_boa_left_04")
    Value=(Rotation=(X=0,Y=-0.0262,Z=0,W=0.9997),Translation=(X=0,Y=0,Z=-1),Scale3D=(X=1,Y=1,Z=1))
    Weight=1
    bParentInitial=False
    bPropagateToChildren=True
   ) {template="Set Relative Transform::Execute(in Child,in Parent,in bParentInitial,in Value,in Weight,in bPropagateToChildren)"}
   > &RigUnit_SetRelativeTransformForItem(
    Child=(Type=Bone,Name="dyn_boa_left_05")
    Parent=(Type=Bone,Name="dyn_boa_left_05")
    Value=(Rotation=(X=0,Y=-0.0872,Z=0,W=0.9962),Translation=(X=0,Y=0,Z=0),Scale3D=(X=1,Y=1,Z=1))
    Weight=1
    bParentInitial=False
    bPropagateToChildren=True
   ) {template="Set Relative Transform::Execute(in Child,in Parent,in bParentInitial,in Value,in Weight,in bPropagateToChildren)"}
   n2 = &RigUnit_SphericalPoseReader(
    ActiveRegionScaleFactors=(PositiveWidth=1,NegativeWidth=1,PositiveHeight=1,NegativeHeight=1)
    ActiveRegionSize=0.1
    Debug=(bDrawDebug=false,DebugScale=25,DebugSegments=20,DebugThickness=0.25)
    DriverAxis=(X=1,Y=0,Z=0)
    DriverItem=(Type=Bone,Name="upperarm_l")
    FalloffRegionScaleFactors=(PositiveWidth=1,NegativeWidth=1,PositiveHeight=1,NegativeHeight=1)
    FalloffSize=0.25
    OptionalParentItem=(Type=Bone)
    RotationOffset=(X=-30,Y=-90,Z=0)
    ExecutePin=@n1.B
   )
   > n3 = &RigUnit_SphericalPoseReader(
    ActiveRegionScaleFactors=(PositiveWidth=1,NegativeWidth=1,PositiveHeight=1,NegativeHeight=1)
    ActiveRegionSize=0.1
    Debug=(bDrawDebug=false,DebugScale=25,DebugSegments=20,DebugThickness=0.25)
    DriverAxis=(X=-1,Y=0,Z=0)
    DriverItem=(Type=Bone,Name="upperarm_r")
    FalloffRegionScaleFactors=(PositiveWidth=1,NegativeWidth=1,PositiveHeight=1,NegativeHeight=1)
    FalloffSize=0.25
    FlipHeightScaling=true
    FlipWidthScaling=false
    OptionalParentItem=(Type=Bone)
    RotationOffset=(X=30,Y=110,Z=0)
   )
   n4 = &RigVMFunction_MathDoubleMul(B=@n2.OutputParam, A=@n3.OutputParam)
   > &RigUnit_SetRelativeTransformForItem(
    Child=(Type=Bone,Name="def_boa_left_01")
    Parent=(Type=Bone,Name="def_boa_left_01")
    Value=(Rotation=(X=0,Y=0,Z=0,W=1),Translation=(X=0,Y=2,Z=0),Scale3D=(X=1,Y=1,Z=1))
    bParentInitial=False
    bPropagateToChildren=True
    Weight=@n4.Result
   ) {template="Set Relative Transform::Execute(in Child,in Parent,in bParentInitial,in Value,in Weight,in bPropagateToChildren)"}
   > &RigUnit_SetRelativeTransformForItem(
    Child=(Type=Bone,Name="def_boa_right_01")
    Parent=(Type=Bone,Name="def_boa_right_01")
    Value=(Rotation=(X=0,Y=0,Z=0,W=1),Translation=(X=0,Y=-2,Z=0),Scale3D=(X=1,Y=1,Z=1))
    bParentInitial=False
    bPropagateToChildren=True
    Weight=@n4.Result
   ) {template="Set Relative Transform::Execute(in Child,in Parent,in bParentInitial,in Value,in Weight,in bPropagateToChildren)"}
   RigVMFunction_ControlFlowBranch = &RigVMFunction_ControlFlowBranch(ExecuteContext=@RigVMFunction_Sequence.C, Condition=RigVMVariableNode() {variable=crouching}.Value)
   n5 = &RigVMFunction_Sequence(ExecuteContext=@RigVMFunction_ControlFlowBranch.False)
   n6 = &RigUnit_SphericalPoseReader(
    ActiveRegionScaleFactors=(PositiveWidth=2.5,NegativeWidth=3,PositiveHeight=3,NegativeHeight=3)
    ActiveRegionSize=0.1
    Debug=(bDrawDebug=false,DebugScale=25,DebugSegments=20,DebugThickness=0.25)
    DriverAxis=(X=1,Y=0,Z=0)
    DriverItem=(Type=Bone,Name="upperarm_l")
    FalloffRegionScaleFactors=(PositiveWidth=1,NegativeWidth=1,PositiveHeight=1,NegativeHeight=1)
    FalloffSize=0.5
    OptionalParentItem=(Type=Bone,Name="clavicle_l")
    RotationOffset=(X=0,Y=0,Z=0)
    ExecutePin=@n5.A
   )
   > &RigUnit_SetRelativeTransformForItem(
    Child=(Type=Bone,Name="def_boa_right_04")
    Parent=(Type=Bone,Name="def_boa_right_04")
    Value=(Rotation=(X=0,Y=0,Z=0,W=1),Translation=(X=0,Y=3,Z=0),Scale3D=(X=1,Y=1,Z=1))
    bParentInitial=False
    bPropagateToChildren=True
    Weight=@n6.OutputParam
   ) {template="Set Relative Transform::Execute(in Child,in Parent,in bParentInitial,in Value,in Weight,in bPropagateToChildren)"}
   > &RigUnit_SetRelativeTransformForItem(
    Child=(Type=Bone,Name="def_boa_left_04")
    Parent=(Type=Bone,Name="def_boa_left_04")
    Value=(Rotation=(X=0.0402,Y=-0.0773,Z=0.46,W=0.8836),Translation=(X=0,Y=2,Z=1),Scale3D=(X=1,Y=1,Z=1))
    bParentInitial=False
    bPropagateToChildren=True
    Weight=@n6.OutputParam
   ) {template="Set Relative Transform::Execute(in Child,in Parent,in bParentInitial,in Value,in Weight,in bPropagateToChildren)"}
   > n7 = &RigUnit_SphericalPoseReader(
    ActiveRegionScaleFactors=(PositiveWidth=2.5,NegativeWidth=3,PositiveHeight=3,NegativeHeight=3)
    ActiveRegionSize=0.1
    Debug=(bDrawDebug=false,DebugScale=25,DebugSegments=20,DebugThickness=0.25)
    DriverAxis=(X=1,Y=0,Z=0)
    DriverItem=(Type=Bone,Name="upperarm_l")
    FalloffRegionScaleFactors=(PositiveWidth=1,NegativeWidth=1,PositiveHeight=1,NegativeHeight=1)
    FalloffSize=0.5
    OptionalParentItem=(Type=Bone,Name="clavicle_l")
    RotationOffset=(X=0,Y=0,Z=0)
   )
   > n8 = &RigUnit_SphericalPoseReader(
    ActiveRegionScaleFactors=(PositiveWidth=2.5,NegativeWidth=3,PositiveHeight=3,NegativeHeight=3)
    ActiveRegionSize=0.07
    Debug=(bDrawDebug=false,DebugScale=25,DebugSegments=20,DebugThickness=0.25)
    DriverAxis=(X=-1,Y=0,Z=0)
    DriverItem=(Type=Bone,Name="upperarm_r")
    FalloffRegionScaleFactors=(PositiveWidth=1,NegativeWidth=1,PositiveHeight=1,NegativeHeight=1)
    FalloffSize=0.45
    OptionalParentItem=(Type=Bone,Name="clavicle_l")
    RotationOffset=(X=-90,Y=0,Z=60)
   )
   Multiply = &RigVMFunction_MathDoubleMul(B=@n7.OutputParam, A=@n8.OutputParam)
   > &RigUnit_SetRelativeTransformForItem(
    Child=(Type=Bone,Name="def_boa_right_04")
    Parent=(Type=Bone,Name="def_boa_right_04")
    Value=(Rotation=(X=0,Y=0,Z=-0.2588,W=0.9659),Translation=(X=0,Y=0,Z=0),Scale3D=(X=1,Y=1,Z=1))
    bParentInitial=False
    bPropagateToChildren=True
    Weight=@Multiply.Result
   ) {template="Set Relative Transform::Execute(in Child,in Parent,in bParentInitial,in Value,in Weight,in bPropagateToChildren)"}
   > &RigUnit_SetRelativeTransformForItem(
    Child=(Type=Bone,Name="def_boa_left_04")
    Parent=(Type=Bone,Name="def_boa_left_04")
    Value=(Rotation=(X=0.0227,Y=0.1285,Z=-0.1722,W=0.9764),Translation=(X=0,Y=2,Z=0),Scale3D=(X=1,Y=1,Z=1))
    bParentInitial=False
    bPropagateToChildren=True
    Weight=@Multiply.Result
   ) {template="Set Relative Transform::Execute(in Child,in Parent,in bParentInitial,in Value,in Weight,in bPropagateToChildren)"}
   > n9 = &RigUnit_SphericalPoseReader(
    ActiveRegionScaleFactors=(PositiveWidth=0.5,NegativeWidth=3,PositiveHeight=1.5,NegativeHeight=1.5)
    ActiveRegionSize=0.1
    Debug=(bDrawDebug=false,DebugScale=25,DebugSegments=20,DebugThickness=0.25)
    DriverAxis=(X=-1,Y=0,Z=0)
    DriverItem=(Type=Bone,Name="upperarm_r")
    FalloffRegionScaleFactors=(PositiveWidth=1,NegativeWidth=1,PositiveHeight=1,NegativeHeight=1)
    FalloffSize=0.35
    FlipWidthScaling=true
    OptionalParentItem=(Type=Bone)
    RotationOffset=(X=90,Y=180,Z=180)
   )
   n10 = &RigVMFunction_AlphaInterp(InterpSpeedDecreasing=10, InterpSpeedIncreasing=9, Scale=1, bClampResult=false, Value=@n9.OutputParam)
   > &RigUnit_SetRelativeTransformForItem(
    Child=(Type=Bone,Name="def_boa_right_03")
    Parent=(Type=Bone,Name="def_boa_right_03")
    Value=(Rotation=(X=-0.003,Y=0.0871,Z=0.0348,W=0.9956),Translation=(X=5,Y=0,Z=0),Scale3D=(X=1,Y=1,Z=1))
    bParentInitial=False
    bPropagateToChildren=True
    Weight=@n10.Result
   ) {template="Set Relative Transform::Execute(in Child,in Parent,in bParentInitial,in Value,in Weight,in bPropagateToChildren)"}
   > n11 = &RigUnit_SphericalPoseReader(
    ActiveRegionScaleFactors=(PositiveWidth=1,NegativeWidth=1,PositiveHeight=1,NegativeHeight=1)
    ActiveRegionSize=0.1
    Debug=(bDrawDebug=false,bDraw2D=False,bDrawLocalAxes=False,DebugScale=25,DebugSegments=20,DebugThickness=0.25)
    DriverAxis=(X=1,Y=0,Z=0)
    DriverItem=(Type=Bone,Name="upperarm_l")
    FalloffRegionScaleFactors=(PositiveWidth=1,NegativeWidth=1.4,PositiveHeight=1.5,NegativeHeight=1)
    FalloffSize=0.3
    FlipHeightScaling=False
    FlipWidthScaling=False
    OptionalParentItem=(Type=Bone,Name="None")
    RotationOffset=(X=10,Y=-140,Z=0)
   )
   > n12 = &RigUnit_SphericalPoseReader(
    ActiveRegionScaleFactors=(PositiveWidth=1,NegativeWidth=1,PositiveHeight=1,NegativeHeight=1)
    ActiveRegionSize=0.1
    Debug=(bDrawDebug=false,bDraw2D=False,bDrawLocalAxes=False,DebugScale=25,DebugSegments=20,DebugThickness=0.25)
    DriverAxis=(X=-1,Y=0,Z=0)
    DriverItem=(Type=Bone,Name="upperarm_r")
    FalloffRegionScaleFactors=(PositiveWidth=1,NegativeWidth=1.4,PositiveHeight=1,NegativeHeight=1.5)
    FalloffSize=0.3
    FlipHeightScaling=False
    FlipWidthScaling=False
    OptionalParentItem=(Type=Bone,Name="None")
    RotationOffset=(X=-32,Y=20,Z=0)
   )
   > &RigUnit_SetRelativeTransformForItem(
    Child=(Type=Bone,Name="def_boa_right_04")
    Parent=(Type=Bone,Name="def_boa_right_04")
    Value=(Rotation=(X=0.056,Y=-0.2527,Z=0.2091,W=0.943),Translation=(X=0,Y=0,Z=0),Scale3D=(X=1,Y=1,Z=1))
    bParentInitial=False
    bPropagateToChildren=True
    Weight=&RigVMFunction_MathDoubleMul(B=@n11.OutputParam, A=@n12.OutputParam).Result
   ) {template="Set Relative Transform::Execute(in Child,in Parent,in bParentInitial,in Value,in Weight,in bPropagateToChildren)"}
   n13 = &RigVMFunction_AlphaInterp(
    InterpSpeedDecreasing=10
    InterpSpeedIncreasing=10
    Scale=1
    bClampResult=false
    Value=&RigVMFunction_MathBoolToFloat(Value=RigVMVariableNode() {variable=targeting}.Value).Result
   )
   &RigUnit_SetRelativeTransformForItem(
    Child=(Type=Bone,Name="def_boa_right_02")
    Parent=(Type=Bone,Name="def_boa_right_02")
    Value=(Rotation=(X=0,Y=0,Z=0,W=1),Translation=(X=0,Y=3.5,Z=0.7),Scale3D=(X=1,Y=1,Z=1))
    bParentInitial=False
    bPropagateToChildren=True
    Weight=@n13.Result
    ExecutePin=@n5.B
   ) {template="Set Relative Transform::Execute(in Child,in Parent,in bParentInitial,in Value,in Weight,in bPropagateToChildren)"}
   > &RigUnit_SetRelativeTransformForItem(
    Child=(Type=Bone,Name="def_boa_left_04")
    Parent=(Type=Bone,Name="def_boa_left_04")
    Value=(Rotation=(X=0.0127,Y=0.2614,Z=0.1483,W=-0.9537),Translation=(X=0,Y=0,Z=0),Scale3D=(X=1,Y=1,Z=1))
    bParentInitial=False
    bPropagateToChildren=True
    Weight=&RigVMFunction_MathBoolToFloat(Value=RigVMVariableNode() {variable=DBNO}.Value).Result
   ) {template="Set Relative Transform::Execute(in Child,in Parent,in bParentInitial,in Value,in Weight,in bPropagateToChildren)"}
   &RigUnit_SetRelativeTransformForItem(
    Child=(Type=Bone,Name="def_boa_right_04")
    Parent=(Type=Bone,Name="def_boa_right_04")
    Value=(Rotation=(X=-0.0151,Y=0.0858,Z=0.173,W=0.9811),Translation=(X=0,Y=0,Z=0),Scale3D=(X=1,Y=1,Z=1))
    Weight=1
    bParentInitial=False
    bPropagateToChildren=True
    ExecutePin=@RigVMFunction_ControlFlowBranch.True
   ) {template="Set Relative Transform::Execute(in Child,in Parent,in bParentInitial,in Value,in Weight,in bPropagateToChildren)"}
   > n14 = &RigUnit_SphericalPoseReader(
    ActiveRegionScaleFactors=(PositiveWidth=1,NegativeWidth=1,PositiveHeight=1,NegativeHeight=1)
    ActiveRegionSize=0.12
    Debug=(bDrawDebug=false,bDraw2D=False,bDrawLocalAxes=False,DebugScale=25,DebugSegments=20,DebugThickness=0.25)
    DriverAxis=(X=-1,Y=0,Z=0)
    DriverItem=(Type=Bone,Name="thigh_l")
    FalloffRegionScaleFactors=(PositiveWidth=1.6,NegativeWidth=1.6,PositiveHeight=1.6,NegativeHeight=1.6)
    FalloffSize=0.2
    FlipHeightScaling=False
    FlipWidthScaling=False
    OptionalParentItem=(Type=Bone,Name="None")
    RotationOffset=(X=120,Y=0,Z=-20)
   )
   > RigUnit_SetRelativeTransformForItem = &RigUnit_SetRelativeTransformForItem(
    Child=(Type=Bone,Name="def_boa_left_04")
    Parent=(Type=Bone,Name="def_boa_left_04")
    Value=(Rotation=(X=-0.0057,Y=-0.1304,Z=-0.0432,W=0.9905),Translation=(X=0,Y=0,Z=0),Scale3D=(X=1,Y=1,Z=1))
    bParentInitial=False
    bPropagateToChildren=True
    Weight=@n14.OutputParam
   ) {template="Set Relative Transform::Execute(in Child,in Parent,in bParentInitial,in Value,in Weight,in bPropagateToChildren)"}
   > &RigUnit_SetRelativeTransformForItem(
    Child=(Type=Bone,Name="def_boa_right_04")
    Parent=(Type=Bone,Name="def_boa_right_04")
    Value=(Rotation=(X=-0.0076,Y=-0.1735,Z=-0.043,W=0.9839),Translation=(X=0,Y=0,Z=0),Scale3D=(X=1,Y=1,Z=1))
    Weight=1
    bParentInitial=False
    bPropagateToChildren=True
   ) {template="Set Relative Transform::Execute(in Child,in Parent,in bParentInitial,in Value,in Weight,in bPropagateToChildren)"}
   > n15 = &RigUnit_SphericalPoseReader(
    ActiveRegionScaleFactors=(PositiveWidth=1,NegativeWidth=1,PositiveHeight=1,NegativeHeight=1)
    ActiveRegionSize=0.1
    Debug=(bDrawDebug=false,DebugScale=25,DebugSegments=20,DebugThickness=0.25)
    DriverAxis=(X=1,Y=0,Z=0)
    DriverItem=(Type=Bone,Name="upperarm_l")
    FalloffRegionScaleFactors=(PositiveWidth=1,NegativeWidth=1,PositiveHeight=1,NegativeHeight=1)
    FalloffSize=0.3
    OptionalParentItem=(Type=Bone,Name="spine_05")
    RotationOffset=(X=7.0096,Y=-141.722,Z=16.681)
   )
   > &RigUnit_SetRelativeTransformForItem(
    Child=(Type=Bone,Name="def_boa_right_04")
    Parent=(Type=Bone,Name="def_boa_right_04")
    Value=(Rotation=(X=0.352,Y=-0.2211,Z=0.0142,W=-0.9094),Translation=(X=0,Y=0,Z=5),Scale3D=(X=1,Y=1,Z=1))
    bParentInitial=False
    bPropagateToChildren=True
    Weight=@n15.OutputParam
   ) {template="Set Relative Transform::Execute(in Child,in Parent,in bParentInitial,in Value,in Weight,in bPropagateToChildren)"}
   > n16 = &RigUnit_SphericalPoseReader(
    ActiveRegionScaleFactors=(PositiveWidth=2.5,NegativeWidth=3,PositiveHeight=3,NegativeHeight=3)
    ActiveRegionSize=0.1
    Debug=(bDrawDebug=false,DebugScale=25,DebugSegments=20,DebugThickness=0.25)
    DriverAxis=(X=1,Y=0,Z=0)
    DriverItem=(Type=Bone,Name="upperarm_l")
    FalloffRegionScaleFactors=(PositiveWidth=1,NegativeWidth=1,PositiveHeight=1,NegativeHeight=1)
    FalloffSize=0.5
    OptionalParentItem=(Type=Bone,Name="clavicle_l")
    RotationOffset=(X=0,Y=0,Z=0)
   )
   > &RigUnit_SetRelativeTransformForItem(
    Child=(Type=Bone,Name="def_boa_right_04")
    Parent=(Type=Bone,Name="def_boa_right_04")
    Value=(Rotation=(X=0.0113,Y=-0.2586,Z=0.0421,W=0.965),Translation=(X=0,Y=0,Z=-1),Scale3D=(X=1,Y=1,Z=1))
    bParentInitial=False
    bPropagateToChildren=True
    Weight=@n16.OutputParam
   ) {template="Set Relative Transform::Execute(in Child,in Parent,in bParentInitial,in Value,in Weight,in bPropagateToChildren)"}
   > &RigUnit_SetRelativeTransformForItem(
    Child=(Type=Bone,Name="def_boa_left_04")
    Parent=(Type=Bone,Name="def_boa_left_04")
    Value=(Rotation=(X=-0.0665,Y=0.1604,Z=0.3769,W=0.9098),Translation=(X=0,Y=0,Z=-2),Scale3D=(X=1,Y=1,Z=1))
    bParentInitial=False
    bPropagateToChildren=True
    Weight=@n16.OutputParam
   ) {template="Set Relative Transform::Execute(in Child,in Parent,in bParentInitial,in Value,in Weight,in bPropagateToChildren)"}
   > n17 = &RigUnit_SphericalPoseReader(
    ActiveRegionScaleFactors=(PositiveWidth=2.5,NegativeWidth=3,PositiveHeight=3,NegativeHeight=3)
    ActiveRegionSize=0.1
    Debug=(bDrawDebug=false,DebugScale=25,DebugSegments=20,DebugThickness=0.25)
    DriverAxis=(X=1,Y=0,Z=0)
    DriverItem=(Type=Bone,Name="upperarm_l")
    FalloffRegionScaleFactors=(PositiveWidth=1,NegativeWidth=1,PositiveHeight=1,NegativeHeight=1)
    FalloffSize=0.5
    OptionalParentItem=(Type=Bone,Name="clavicle_l")
    RotationOffset=(X=0,Y=0,Z=0)
   )
   > n18 = &RigUnit_SphericalPoseReader(
    ActiveRegionScaleFactors=(PositiveWidth=2.5,NegativeWidth=3,PositiveHeight=3,NegativeHeight=3)
    ActiveRegionSize=0.07
    Debug=(bDrawDebug=false,DebugScale=25,DebugSegments=20,DebugThickness=0.25)
    DriverAxis=(X=-1,Y=0,Z=0)
    DriverItem=(Type=Bone,Name="upperarm_r")
    FalloffRegionScaleFactors=(PositiveWidth=1,NegativeWidth=1,PositiveHeight=1,NegativeHeight=1)
    FalloffSize=0.45
    OptionalParentItem=(Type=Bone,Name="clavicle_l")
    RotationOffset=(X=-90,Y=0,Z=60)
   )
   n19 = &RigVMFunction_MathDoubleMul(B=@n17.OutputParam, A=@n18.OutputParam)
   > &RigUnit_SetRelativeTransformForItem(
    Child=(Type=Bone,Name="def_boa_right_04")
    Parent=(Type=Bone,Name="def_boa_right_04")
    Value=(Rotation=(X=0,Y=0.1736,Z=0,W=0.9848),Translation=(X=0,Y=0,Z=0),Scale3D=(X=1,Y=1,Z=1))
    bParentInitial=False
    bPropagateToChildren=True
    Weight=@n19.Result
   ) {template="Set Relative Transform::Execute(in Child,in Parent,in bParentInitial,in Value,in Weight,in bPropagateToChildren)"}
   > &RigUnit_SetRelativeTransformForItem(
    Child=(Type=Bone,Name="def_boa_right_03")
    Parent=(Type=Bone,Name="def_boa_right_03")
    Value=(Rotation=(X=-0.2588,Y=0,Z=0,W=0.9659),Translation=(X=0,Y=2,Z=-1),Scale3D=(X=1,Y=1,Z=1))
    bParentInitial=False
    bPropagateToChildren=True
    Weight=@n19.Result
   ) {template="Set Relative Transform::Execute(in Child,in Parent,in bParentInitial,in Value,in Weight,in bPropagateToChildren)"}
   > n20 = &RigUnit_SphericalPoseReader(
    ActiveRegionScaleFactors=(PositiveWidth=1.4,NegativeWidth=1,PositiveHeight=1,NegativeHeight=1.2)
    ActiveRegionSize=0.1
    Debug=(bDrawDebug=True,bDraw2D=False,bDrawLocalAxes=False,DebugScale=25,DebugSegments=20,DebugThickness=0.25)
    DriverAxis=(X=-1,Y=0,Z=0)
    DriverItem=(Type=Bone,Name="upperarm_r")
    FalloffRegionScaleFactors=(PositiveWidth=1,NegativeWidth=1,PositiveHeight=1,NegativeHeight=1)
    FalloffSize=0.2
    FlipHeightScaling=False
    FlipWidthScaling=False
    OptionalParentItem=(Type=Bone,Name="None")
    RotationOffset=(X=0,Y=40,Z=20)
   )
   > n21 = &RigUnit_SphericalPoseReader(
    ActiveRegionScaleFactors=(PositiveWidth=1.2,NegativeWidth=1.5,PositiveHeight=2.8,NegativeHeight=1.2)
    ActiveRegionSize=0.1
    Debug=(bDrawDebug=True,bDraw2D=False,bDrawLocalAxes=False,DebugScale=25,DebugSegments=20,DebugThickness=0.25)
    DriverAxis=(X=1,Y=0,Z=0)
    DriverItem=(Type=Bone,Name="upperarm_l")
    FalloffRegionScaleFactors=(PositiveWidth=0.7,NegativeWidth=0.7,PositiveHeight=3,NegativeHeight=0.7)
    FalloffSize=0.3
    FlipHeightScaling=False
    FlipWidthScaling=False
    OptionalParentItem=(Type=Bone,Name="None")
    RotationOffset=(X=0,Y=-105,Z=0)
   )
   n22 = &RigVMFunction_AlphaInterp(
    InterpSpeedDecreasing=2
    InterpSpeedIncreasing=2
    Scale=1
    bClampResult=False
    bInterpResult=False
    Value=&RigVMFunction_MathDoubleMul(B=@n20.OutputParam, A=@n21.OutputParam).Result
   )
   > &RigUnit_SetRelativeTransformForItem(
    Child=(Type=Bone,Name="def_boa_right_04")
    Parent=(Type=Bone,Name="def_boa_right_04")
    Value=(Rotation=(X=-0.073,Y=0.5047,Z=0.1223,W=-0.8515),Translation=(X=0,Y=0,Z=0),Scale3D=(X=1,Y=1,Z=1))
    bParentInitial=False
    bPropagateToChildren=True
    Weight=@n22.Result
   ) {template="Set Relative Transform::Execute(in Child,in Parent,in bParentInitial,in Value,in Weight,in bPropagateToChildren)"}
   RigVMFunction_Sequence_1 = RigVMAggregateNode(ExecuteContext=@RigVMFunction_Sequence.D)
   RigVMFunction_MathBoolToFloat = &RigVMFunction_MathBoolToFloat(Value=RigVMVariableNode() {variable=Skidive floating}.Value)
   AlphaInterp = &RigVMFunction_AlphaInterp(InterpSpeedDecreasing=5, InterpSpeedIncreasing=5, Scale=1, bClampResult=false, Value=@RigVMFunction_MathBoolToFloat.Result)
   &RigUnit_SetRelativeTransformForItem(
    Child=(Type=Bone,Name="def_boa_right_04")
    Parent=(Type=Bone,Name="def_boa_right_04")
    Value=(Rotation=(X=0,Y=0,Z=-0.4617,W=0.887),Translation=(X=0,Y=0,Z=0),Scale3D=(X=1,Y=1,Z=1))
    bParentInitial=False
    bPropagateToChildren=True
    Weight=@AlphaInterp.Result
    ExecutePin=@RigVMFunction_Sequence_1.A
   ) {template="Set Relative Transform::Execute(in Child,in Parent,in bParentInitial,in Value,in Weight,in bPropagateToChildren)"}
   > &RigUnit_SetRelativeTransformForItem(
    Child=(Type=Bone,Name="def_boa_left_04")
    Parent=(Type=Bone,Name="def_boa_left_04")
    Value=(Rotation=(X=0.0235,Y=0.2164,Z=-0.0718,W=0.9734),Translation=(X=0,Y=0,Z=0),Scale3D=(X=1,Y=1,Z=1))
    bParentInitial=False
    bPropagateToChildren=True
    Weight=@AlphaInterp.Result
   ) {template="Set Relative Transform::Execute(in Child,in Parent,in bParentInitial,in Value,in Weight,in bPropagateToChildren)"}
   n23 = &RigVMFunction_ControlFlowBranch(ExecuteContext=@RigVMFunction_Sequence_1.C, Condition=RigVMVariableNode() {variable=playing emote}.Value)
   n24 = &RigUnit_SphericalPoseReader(
    ActiveRegionScaleFactors=(PositiveWidth=2,NegativeWidth=2,PositiveHeight=1,NegativeHeight=1)
    ActiveRegionSize=0.125
    Debug=(bDrawDebug=false,bDrawLocalAxes=false,DebugScale=25,DebugSegments=20,DebugThickness=0.25)
    DriverAxis=(X=1,Y=0,Z=0)
    DriverItem=(Type=Bone,Name="upperarm_l")
    FalloffRegionScaleFactors=(PositiveWidth=0.5,NegativeWidth=0.5,PositiveHeight=0.5,NegativeHeight=0.5)
    FalloffSize=0.5
    OptionalParentItem=(Type=Bone,Name="clavicle_l")
    RotationOffset=(X=50,Y=-30,Z=-90)
    ExecutePin=@n23.True
   )
   > n25 = &RigUnit_SphericalPoseReader(
    ActiveRegionScaleFactors=(PositiveWidth=2,NegativeWidth=2,PositiveHeight=1,NegativeHeight=1)
    ActiveRegionSize=0.125
    Debug=(bDrawDebug=false,bDrawLocalAxes=false,DebugScale=25,DebugSegments=20,DebugThickness=0.25)
    DriverAxis=(X=-1,Y=0,Z=0)
    DriverItem=(Type=Bone,Name="upperarm_r")
    FalloffRegionScaleFactors=(PositiveWidth=0.5,NegativeWidth=0.5,PositiveHeight=0.5,NegativeHeight=0.5)
    FalloffSize=0.5
    OptionalParentItem=(Type=Bone,Name="clavicle_l")
    RotationOffset=(X=-35,Y=160,Z=100)
   )
   n26 = &RigVMFunction_MathDoubleMul(B=@n24.OutputParam, A=@n25.OutputParam)
   > &RigUnit_SetRelativeTransformForItem(
    Child=(Type=Bone,Name="def_boa_right_04")
    Parent=(Type=Bone,Name="def_boa_right_04")
    Value=(Rotation=(X=-0.0338,Y=-0.2566,Z=-0.1261,W=0.9577),Translation=(X=0,Y=0,Z=0),Scale3D=(X=1,Y=1,Z=1))
    bParentInitial=False
    bPropagateToChildren=True
    Weight=@n26.Result
   ) {template="Set Relative Transform::Execute(in Child,in Parent,in bParentInitial,in Value,in Weight,in bPropagateToChildren)"}
   > &RigUnit_SetRelativeTransformForItem(
    Child=(Type=Bone,Name="def_boa_left_04")
    Parent=(Type=Bone,Name="def_boa_left_04")
    Value=(Rotation=(X=0.0076,Y=-0.0868,Z=0.0868,W=0.9924),Translation=(X=0,Y=0,Z=0),Scale3D=(X=1,Y=1,Z=1))
    bParentInitial=False
    bPropagateToChildren=True
    Weight=@n26.Result
   ) {template="Set Relative Transform::Execute(in Child,in Parent,in bParentInitial,in Value,in Weight,in bPropagateToChildren)"}
   n27 = &RigVMFunction_AlphaInterp(
    InterpSpeedDecreasing=5
    InterpSpeedIncreasing=5
    Scale=1
    bClampResult=false
    Value=&RigVMFunction_MathBoolToFloat(Value=RigVMVariableNode() {variable=Parachute opened}.Value).Result
   )
   &RigUnit_SetRelativeTransformForItem(
    Child=(Type=Bone,Name="def_boa_right_04")
    Parent=(Type=Bone,Name="def_boa_right_04")
    Value=(Rotation=(X=-0.0226,Y=-0.2578,Z=-0.0842,W=0.9623),Translation=(X=0,Y=0,Z=0),Scale3D=(X=1,Y=1,Z=1))
    bParentInitial=False
    bPropagateToChildren=True
    Weight=@n27.Result
    ExecutePin=@RigVMFunction_Sequence_1.B
   ) {template="Set Relative Transform::Execute(in Child,in Parent,in bParentInitial,in Value,in Weight,in bPropagateToChildren)"}
   > &RigUnit_SetRelativeTransformForItem(
    Child=(Type=Bone,Name="def_boa_left_04")
    Parent=(Type=Bone,Name="def_boa_left_04")
    Value=(Rotation=(X=0,Y=0,Z=0.1736,W=0.9848),Translation=(X=0,Y=0,Z=0),Scale3D=(X=1,Y=1,Z=1))
    bParentInitial=False
    bPropagateToChildren=True
    Weight=@n27.Result
   ) {template="Set Relative Transform::Execute(in Child,in Parent,in bParentInitial,in Value,in Weight,in bPropagateToChildren)"}
   > &RigUnit_SetRelativeTransformForItem(
    Child=(Type=Bone,Name="def_boa_right_02")
    Parent=(Type=Bone,Name="def_boa_right_02")
    Value=(Rotation=(X=0,Y=0,Z=0.1736,W=0.9848),Translation=(X=0,Y=2,Z=0),Scale3D=(X=1,Y=1,Z=1))
    bParentInitial=False
    bPropagateToChildren=True
    Weight=@n27.Result
   ) {template="Set Relative Transform::Execute(in Child,in Parent,in bParentInitial,in Value,in Weight,in bPropagateToChildren)"}
   > &RigUnit_SetRelativeTransformForItem(
    Child=(Type=Bone,Name="def_boa_left_02")
    Parent=(Type=Bone,Name="def_boa_left_02")
    Value=(Rotation=(X=0,Y=0,Z=-0.0872,W=0.9962),Translation=(X=0,Y=-2,Z=0),Scale3D=(X=1,Y=1,Z=1))
    bParentInitial=False
    bPropagateToChildren=True
    Weight=@n27.Result
   ) {template="Set Relative Transform::Execute(in Child,in Parent,in bParentInitial,in Value,in Weight,in bPropagateToChildren)"}
   SplineIK = RigVMFunctionReferenceNode(
    Primary Axis=(X=1,Y=0,Z=0)
    Samples Per Segment=6
    Secondary Spline Direction=(X=0,Y=1,Z=0)
    Spline Mode=BSpline
    Stretch=false
    Up Axis=(X=0,Y=0,Z=0)
    debug=false
    Controls=RigVMVariableNode() {variable=boa_left_ctrl}.Value
    Bones=RigVMVariableNode() {variable=boa_left_joints}.Value
    ExecuteContext=@RigVMFunction_Sequence.A
   )
   > SplineIK_1 = RigVMFunctionReferenceNode(
    Primary Axis=(X=1,Y=0,Z=0)
    Samples Per Segment=6
    Secondary Spline Direction=(X=0,Y=1,Z=0)
    Spline Mode=BSpline
    Stretch=true
    Up Axis=(X=0,Y=0,Z=0)
    debug=false
    Controls=RigVMVariableNode() {variable=boa_right_ctrl}.Value
    Bones=RigVMVariableNode() {variable=boa_right_joints}.Value
   )
   > RigVMFunctionReturnNode()
   &RigUnit_SetRelativeTransformForItem(
    Child=(Type=Bone,Name="def_boa_right_04")
    ExecutePin=()
    Parent=(Type=Bone,Name="def_boa_right_04")
    Value=(Rotation=(X=0.132,Y=-0.7489,Z=0.1128,W=0.6396),Translation=(X=-4.0074,Y=3.6932,Z=1.6689),Scale3D=(X=1,Y=1,Z=1))
    bParentInitial=False
    bPropagateToChildren=True
    Weight=@n9.OutputParam
   ) {template="Set Relative Transform::Execute(in Child,in Parent,in bParentInitial,in Value,in Weight,in bPropagateToChildren)"}
  ]}
 RigVMFunctionLibrary {nodes [
   PIN = RigVMCollapseNode(ExecuteContext=())
   boa_solver = RigVMCollapseNode(ExecuteContext=())
  ]}
 RigVMFunction_Sequence_1_ContainedGraph {nodes [
   n1 = RigVMFunctionEntryNode()
   RigVMFunction_Sequence = &RigVMFunction_Sequence(ExecuteContext=@n1.ExecuteContext)
   n2 = &RigVMFunction_Sequence(ExecuteContext=@RigVMFunction_Sequence.B)
   n3 = &RigVMFunction_Sequence(ExecuteContext=@n2.B)
   n4 = &RigVMFunction_Sequence(ExecuteContext=@n3.B)
   RigVMFunctionReturnNode(A=@RigVMFunction_Sequence.A, B=@n2.A, C=@n3.A, D=@n4.A, E=@n4.B)
  ]}
 RigVMFunction_Sequence_2_ContainedGraph {nodes [
   n1 = RigVMFunctionEntryNode()
   n2 = &RigVMFunction_Sequence(ExecuteContext=@n1.ExecuteContext)
   n3 = &RigVMFunction_Sequence(ExecuteContext=@n2.B)
   RigVMFunctionReturnNode(A=@n2.A, B=@n3.A, C=@n3.B)
  ]}
 RigVMFunction_Sequence_2_ContainedGraph {nodes [
   n1 = RigVMFunctionEntryNode()
   n2 = &RigVMFunction_Sequence(ExecuteContext=@n1.ExecuteContext)
   n3 = &RigVMFunction_Sequence(ExecuteContext=@n2.B)
   RigVMFunctionReturnNode(A=@n2.A, B=@n3.A, C=@n3.B)
  ]}
 RigVMModel {nodes [
   RigUnit_BeginExecution = &RigUnit_BeginExecution()
   FN_DeformRig_FNC = RigVMFunctionReferenceNode(
    DeformRig_ Curve="(deformRig_l_upperarm_up_alpha_72_#1=False,deformRig_r_upperarm_up_alpha_73_#2=False,deformRig_l_upperarm_fr_alpha_74_#3=False,deformRig_r_upperarm_fr_alpha_75_CB51B8 ...(truncated)"
    DeformRig_Bool="(upperarm_1_#4=True,elbow_3_#5=True,wrist_7_#6=True,groin_8_#7=True,glute_17_#8=True,knee_9_40C408324CA261 ...(truncated)"
    DeformRig_Debug=(upperarm_debug_18_#4=False,wrist_debug_20_#6=False,lat_debug_26_#9=False,groinGlute_debug_27_#7=False)
    DeformRig_Value="(upperarm_up_18_#10=(Rotation=(X=0,Y=0,Z=0,W=1),Translation=(X=0,Y=0,Z=8),Scale3D=(X=0.7,Y=1,Z=1)),upperarm_fwd_fr_19_#11=(Rotation=(X=0,Y=0,Z=0,W=1),Translation=(X=0,Y=10,Z=0),Scale3D=(X=0.5,Y=1,Z= ...(truncated)"
    ExecuteContext=@RigUnit_BeginExecution.ExecutePin
   )
   RigVMFunction_Sequence = &RigVMFunction_Sequence(ExecuteContext=@FN_DeformRig_FNC.ExecuteContext)
   RigVMFunction_ControlFlowBranch = &RigVMFunction_ControlFlowBranch(ExecuteContext=@RigVMFunction_Sequence.A, Condition=RigVMVariableNode() {variable=RBAN}.Value)
   PIN_1 = RigVMFunctionReferenceNode(ExecuteContext=@RigVMFunction_ControlFlowBranch.False)
   PIN = RigVMFunctionReferenceNode(ExecuteContext=@RigVMFunction_ControlFlowBranch.True)
   > boa_solver = RigVMFunctionReferenceNode()
   RigVMFunction_Sequence_1 = RigVMAggregateNode(ExecuteContext=@RigVMFunction_Sequence.B)
   &RigUnit_SetRelativeTransformForItem(
    Child=(Type=Bone,Name="belt_right_MSTR")
    Parent=(Type=Bone,Name="belt_right_MSTR")
    Value=(Rotation=(X=0.0872,Y=0,Z=0,W=0.9962),Translation=(X=0,Y=0,Z=0),Scale3D=(X=1,Y=1,Z=1))
    Weight=1
    bParentInitial=False
    bPropagateToChildren=True
    ExecutePin=@RigVMFunction_Sequence_1.A
   ) {template="Set Relative Transform::Execute(in Child,in Parent,in bParentInitial,in Value,in Weight,in bPropagateToChildren)"}
   > &RigUnit_SetRelativeTransformForItem(
    Child=(Type=Bone,Name="bag_belt_def")
    Parent=(Type=Bone,Name="bag_belt_def")
    Value=(Rotation=(X=0,Y=0,Z=0,W=-1),Translation=(X=4,Y=-1,Z=0),Scale3D=(X=1,Y=1,Z=1))
    Weight=1
    bParentInitial=False
    bPropagateToChildren=True
   ) {template="Set Relative Transform::Execute(in Child,in Parent,in bParentInitial,in Value,in Weight,in bPropagateToChildren)"}
   > &RigUnit_SetRelativeTransformForItem(
    Child=(Type=Bone,Name="bag_belt_def_back")
    Parent=(Type=Bone,Name="bag_belt_def_back")
    Value=(Rotation=(X=0,Y=0,Z=0,W=1),Translation=(X=0,Y=-1,Z=0),Scale3D=(X=1,Y=1,Z=1))
    Weight=1
    bParentInitial=False
    bPropagateToChildren=True
   ) {template="Set Relative Transform::Execute(in Child,in Parent,in bParentInitial,in Value,in Weight,in bPropagateToChildren)"}
   > n1 = &RigUnit_SphericalPoseReader(
    ActiveRegionScaleFactors=(PositiveWidth=1.9,NegativeWidth=1,PositiveHeight=1.9,NegativeHeight=1.9)
    ActiveRegionSize=0.1
    Debug=(bDrawDebug=false,bDraw2D=False,bDrawLocalAxes=False,DebugScale=25,DebugSegments=20,DebugThickness=0.25)
    DriverAxis=(X=-1,Y=0,Z=0)
    DriverItem=(Type=Bone,Name="thigh_l")
    FalloffRegionScaleFactors=(PositiveWidth=2,NegativeWidth=1.6,PositiveHeight=4,NegativeHeight=3)
    FalloffSize=0.24
    FlipHeightScaling=False
    FlipWidthScaling=False
    OptionalParentItem=(Type=Bone,Name="None")
    RotationOffset=(X=120,Y=0,Z=-10)
   )
   > &RigUnit_SetRelativeTransformForItem(
    Child=(Type=Bone,Name="bag_belt_def_back")
    Parent=(Type=Bone,Name="bag_belt_def_back")
    Value=(Rotation=(X=0,Y=0.0436,Z=0,W=0.999),Translation=(X=0,Y=0,Z=0),Scale3D=(X=1,Y=1,Z=1))
    bParentInitial=False
    bPropagateToChildren=True
    Weight=@n1.OutputParam
   ) {template="Set Relative Transform::Execute(in Child,in Parent,in bParentInitial,in Value,in Weight,in bPropagateToChildren)"}
   > &RigUnit_SetRelativeTransformForItem(
    Child=(Type=Bone,Name="bag_belt_def")
    Parent=(Type=Bone,Name="bag_belt_def")
    Value=(Rotation=(X=0.5,Y=0,Z=0,W=0.866),Translation=(X=2,Y=-1,Z=-4),Scale3D=(X=1,Y=1,Z=1))
    bParentInitial=False
    bPropagateToChildren=True
    Weight=@n1.OutputParam
   ) {template="Set Relative Transform::Execute(in Child,in Parent,in bParentInitial,in Value,in Weight,in bPropagateToChildren)"}
   > &RigUnit_SetRelativeTransformForItem(
    Child=(Type=Bone,Name="bag_belt_front_def")
    Parent=(Type=Bone,Name="bag_belt_front_def")
    Value=(Rotation=(X=0,Y=-0.0872,Z=0,W=0.9962),Translation=(X=4,Y=1,Z=-4),Scale3D=(X=1,Y=1,Z=1))
    bParentInitial=False
    bPropagateToChildren=True
    Weight=@n1.OutputParam
   ) {template="Set Relative Transform::Execute(in Child,in Parent,in bParentInitial,in Value,in Weight,in bPropagateToChildren)"}
   > n2 = &RigUnit_SphericalPoseReader(
    ActiveRegionScaleFactors=(PositiveWidth=1.9,NegativeWidth=1,PositiveHeight=1.9,NegativeHeight=1.9)
    ActiveRegionSize=0.1
    Debug=(bDrawDebug=false,bDraw2D=False,bDrawLocalAxes=False,DebugScale=25,DebugSegments=20,DebugThickness=0.25)
    DriverAxis=(X=1,Y=0,Z=0)
    DriverItem=(Type=Bone,Name="thigh_r")
    FalloffRegionScaleFactors=(PositiveWidth=2,NegativeWidth=1.6,PositiveHeight=4,NegativeHeight=3)
    FalloffSize=0.3
    FlipHeightScaling=False
    FlipWidthScaling=False
    OptionalParentItem=(Type=Bone,Name="None")
    RotationOffset=(X=-75,Y=0,Z=20)
   )
   > &RigUnit_SetRelativeTransformForItem(
    Child=(Type=Bone,Name="belt_back_MSTR")
    Parent=(Type=Bone,Name="belt_back_MSTR")
    Value=(Rotation=(X=0,Y=0,Z=0,W=1),Translation=(X=0,Y=-1,Z=0),Scale3D=(X=1,Y=1,Z=1))
    bParentInitial=False
    bPropagateToChildren=True
    Weight=@n2.OutputParam
   ) {template="Set Relative Transform::Execute(in Child,in Parent,in bParentInitial,in Value,in Weight,in bPropagateToChildren)"}
   > &RigUnit_SetRelativeTransformForItem(
    Child=(Type=Bone,Name="bag_belt_def_right")
    Parent=(Type=Bone,Name="bag_belt_def_right")
    Value=(Rotation=(X=0.1736,Y=0,Z=0,W=0.9848),Translation=(X=0,Y=2,Z=2),Scale3D=(X=1,Y=1,Z=1))
    bParentInitial=False
    bPropagateToChildren=True
    Weight=@n2.OutputParam
   ) {template="Set Relative Transform::Execute(in Child,in Parent,in bParentInitial,in Value,in Weight,in bPropagateToChildren)"}
   > &RigUnit_SetRelativeTransformForItem(
    Child=(Type=Bone,Name="bag_belt_def_back")
    Parent=(Type=Bone,Name="bag_belt_def_back")
    Value=(Rotation=(X=0,Y=0,Z=0,W=1),Translation=(X=-2,Y=0,Z=0),Scale3D=(X=1,Y=1,Z=1))
    bParentInitial=False
    bPropagateToChildren=True
    Weight=@n2.OutputParam
   ) {template="Set Relative Transform::Execute(in Child,in Parent,in bParentInitial,in Value,in Weight,in bPropagateToChildren)"}
   RigVMFunction_MathBoolToFloat = &RigVMFunction_MathBoolToFloat(Value=RigVMVariableNode() {variable=crouching}.Value)
   AlphaInterp = &RigVMFunction_AlphaInterp(InterpSpeedDecreasing=5, InterpSpeedIncreasing=5, Scale=1, bClampResult=false, Value=@RigVMFunction_MathBoolToFloat.Result)
   > &RigUnit_SetRelativeTransformForItem(
    Child=(Type=Bone,Name="bag_belt_def_right")
    Parent=(Type=Bone,Name="bag_belt_def_right")
    Value=(Rotation=(X=0,Y=0,Z=0,W=1),Translation=(X=-0.5,Y=0,Z=0),Scale3D=(X=1,Y=1,Z=1))
    bParentInitial=False
    bPropagateToChildren=True
    Weight=@AlphaInterp.Result
   ) {template="Set Relative Transform::Execute(in Child,in Parent,in bParentInitial,in Value,in Weight,in bPropagateToChildren)"}
   > &RigUnit_SetRelativeTransformForItem(
    Child=(Type=Bone,Name="belt_back_MSTR")
    Parent=(Type=Bone,Name="belt_back_MSTR")
    Value=(Rotation=(X=0,Y=0,Z=0,W=1),Translation=(X=-1,Y=-0.5,Z=0),Scale3D=(X=1,Y=1,Z=1))
    bParentInitial=False
    bPropagateToChildren=True
    Weight=@AlphaInterp.Result
   ) {template="Set Relative Transform::Execute(in Child,in Parent,in bParentInitial,in Value,in Weight,in bPropagateToChildren)"}
   > RigUnit_SetRelativeTransformForItem = &RigUnit_SetRelativeTransformForItem(
    Child=(Type=Bone,Name="belt_back_MSTR")
    Parent=(Type=Bone,Name="belt_back_MSTR")
    Value=(Rotation=(X=0,Y=0,Z=0,W=1),Translation=(X=0,Y=1.2,Z=0),Scale3D=(X=1,Y=1,Z=1))
    bParentInitial=False
    bPropagateToChildren=True
    Weight=@AlphaInterp.Result
   ) {template="Set Relative Transform::Execute(in Child,in Parent,in bParentInitial,in Value,in Weight,in bPropagateToChildren)"}
   n3 = &RigUnit_SphericalPoseReader(
    ActiveRegionScaleFactors=(PositiveWidth=1,NegativeWidth=1,PositiveHeight=1,NegativeHeight=1)
    ActiveRegionSize=0.1
    Debug=(bDrawDebug=false,bDraw2D=False,bDrawLocalAxes=False,DebugScale=25,DebugSegments=20,DebugThickness=0.25)
    DriverAxis=(X=0,Y=1,Z=0)
    DriverItem=(Type=Bone,Name="pelvis")
    FalloffRegionScaleFactors=(PositiveWidth=1,NegativeWidth=1,PositiveHeight=1,NegativeHeight=1)
    FalloffSize=0.2
    FlipHeightScaling=False
    FlipWidthScaling=False
    OptionalParentItem=(Type=Bone,Name="None")
    RotationOffset=(X=110,Y=0,Z=-10)
    ExecutePin=@RigVMFunction_Sequence_1.B
   )
   > &RigUnit_SetRelativeTransformForItem(
    Child=(Type=Bone,Name="bag_belt_def")
    Parent=(Type=Bone,Name="bag_belt_def")
    Value=(Rotation=(X=0,Y=0,Z=0,W=-1),Translation=(X=0,Y=-2,Z=0),Scale3D=(X=1,Y=1,Z=1))
    bParentInitial=False
    bPropagateToChildren=True
    Weight=@n3.OutputParam
   ) {template="Set Relative Transform::Execute(in Child,in Parent,in bParentInitial,in Value,in Weight,in bPropagateToChildren)"}
   > &RigUnit_ParentConstraint(
    AdvancedSettings=(RotationOrderForFilter=XZY)
    Child=(Type=Bone,Name="belt_right_MSTR")
    Parents="((Item=(Type=Bone,Name=\"deform_groin_base_r\"),Weight=0.0005),(Item=(Type=Bone,Name=\"deform_glute_base_r\"),Weight=0.0009),(Item=(Type=Bone,Name=\"thigh_r\"),Weight=0.0017),(Item=(Type=Bone,Name=\"pelvis\"),Weight=0.0039),(Item=(Type=Bone,Name=\"thigh_twist_01_r\" ...(truncated)"
    Weight=1
   )
   > n4 = &RigUnit_SphericalPoseReader(
    ActiveRegionScaleFactors=(PositiveWidth=0.5,NegativeWidth=1,PositiveHeight=1,NegativeHeight=1)
    ActiveRegionSize=0.1
    Debug=(bDrawDebug=false,bDraw2D=False,bDrawLocalAxes=False,DebugScale=25,DebugSegments=20,DebugThickness=0.25)
    DriverAxis=(X=1,Y=0,Z=0)
    DriverItem=(Type=Bone,Name="thigh_r")
    FalloffRegionScaleFactors=(PositiveWidth=1.5,NegativeWidth=1,PositiveHeight=1,NegativeHeight=1)
    FalloffSize=0.2
    FlipHeightScaling=False
    FlipWidthScaling=False
    OptionalParentItem=(Type=Bone,Name="None")
    RotationOffset=(X=-90,Y=0,Z=0)
   )
   > &RigUnit_SetRelativeTransformForItem(
    Child=(Type=Bone,Name="MSTR_right_front_belts")
    Parent=(Type=Bone,Name="MSTR_right_front_belts")
    Value=(Rotation=(X=0.7071,Y=0,Z=0,W=0.7071),Translation=(X=-2,Y=4,Z=11),Scale3D=(X=1,Y=1,Z=1))
    bParentInitial=False
    bPropagateToChildren=True
    Weight=@n4.OutputParam
   ) {template="Set Relative Transform::Execute(in Child,in Parent,in bParentInitial,in Value,in Weight,in bPropagateToChildren)"}
   > n5 = &RigUnit_SphericalPoseReader(
    ActiveRegionScaleFactors=(PositiveWidth=1,NegativeWidth=1,PositiveHeight=1,NegativeHeight=1)
    ActiveRegionSize=0.1
    Debug=(bDrawDebug=false,bDraw2D=False,bDrawLocalAxes=False,DebugScale=25,DebugSegments=20,DebugThickness=0.25)
    DriverAxis=(X=1,Y=0,Z=0)
    DriverItem=(Type=Bone,Name="spine_05")
    FalloffRegionScaleFactors=(PositiveWidth=0.1,NegativeWidth=1,PositiveHeight=1,NegativeHeight=1)
    FalloffSize=0.2
    FlipHeightScaling=False
    FlipWidthScaling=False
    OptionalParentItem=(Type=Bone,Name="None")
    RotationOffset=(X=90,Y=0,Z=-60)
   )
   > &RigUnit_SetRelativeTransformForItem(
    Child=(Type=Bone,Name="dyn_pendant_03")
    Parent=(Type=Bone,Name="dyn_pendant_03")
    Value=(Rotation=(X=0,Y=-0.1736,Z=0,W=0.9848),Translation=(X=-1,Y=0,Z=0),Scale3D=(X=1,Y=1,Z=1))
    bParentInitial=False
    bPropagateToChildren=True
    Weight=@n5.OutputParam
   ) {template="Set Relative Transform::Execute(in Child,in Parent,in bParentInitial,in Value,in Weight,in bPropagateToChildren)"}
   n6 = &RigVMFunction_MathBoolToFloat(Value=RigVMVariableNode() {variable=Skidive floating}.Value)
   > &RigUnit_SetRelativeTransformForItem(
    Child=(Type=Bone,Name="belt_right_MSTR")
    Parent=(Type=Bone,Name="belt_right_MSTR")
    Value=(Rotation=(X=0,Y=0,Z=0,W=1),Translation=(X=-1,Y=-2,Z=0),Scale3D=(X=1,Y=1,Z=1))
    bParentInitial=False
    bPropagateToChildren=True
    Weight=@n6.Result
   ) {template="Set Relative Transform::Execute(in Child,in Parent,in bParentInitial,in Value,in Weight,in bPropagateToChildren)"}
   > &RigUnit_SetRelativeTransformForItem(
    Child=(Type=Bone,Name="bag_belt_def")
    Parent=(Type=Bone,Name="bag_belt_def")
    Value=(Rotation=(X=0,Y=0,Z=0,W=1),Translation=(X=0,Y=-1,Z=0),Scale3D=(X=1,Y=1,Z=1))
    bParentInitial=False
    bPropagateToChildren=True
    Weight=@n6.Result
   ) {template="Set Relative Transform::Execute(in Child,in Parent,in bParentInitial,in Value,in Weight,in bPropagateToChildren)"}
   > &RigUnit_SetRelativeTransformForItem(
    Child=(Type=Bone,Name="MSTR_waist_bag")
    Parent=(Type=Bone,Name="MSTR_waist_bag")
    Value=(Rotation=(X=-0.0872,Y=0,Z=0,W=0.9962),Translation=(X=-1,Y=1,Z=0),Scale3D=(X=1,Y=1,Z=1))
    bParentInitial=False
    bPropagateToChildren=True
    Weight=@n6.Result
   ) {template="Set Relative Transform::Execute(in Child,in Parent,in bParentInitial,in Value,in Weight,in bPropagateToChildren)"}
   > n7 = &RigUnit_SphericalPoseReader(
    ActiveRegionScaleFactors=(PositiveWidth=1,NegativeWidth=1,PositiveHeight=1,NegativeHeight=1)
    ActiveRegionSize=0.1
    Debug=(bDrawDebug=false,bDraw2D=False,bDrawLocalAxes=False,DebugScale=25,DebugSegments=20,DebugThickness=0.25)
    DriverAxis=(X=0,Y=1,Z=0)
    DriverItem=(Type=Bone,Name="pelvis")
    FalloffRegionScaleFactors=(PositiveWidth=1,NegativeWidth=1,PositiveHeight=1,NegativeHeight=1)
    FalloffSize=0.2
    FlipHeightScaling=False
    FlipWidthScaling=False
    OptionalParentItem=(Type=Bone,Name="None")
    RotationOffset=(X=5,Y=0,Z=0)
   )
   > &RigUnit_SetRelativeTransformForItem(
    Child=(Type=Bone,Name="belt_right_MSTR")
    Parent=(Type=Bone,Name="belt_right_MSTR")
    Value=(Rotation=(X=0,Y=0,Z=0,W=1),Translation=(X=0,Y=0,Z=-2),Scale3D=(X=1,Y=1,Z=1))
    bParentInitial=False
    bPropagateToChildren=True
    Weight=@n7.OutputParam
   ) {template="Set Relative Transform::Execute(in Child,in Parent,in bParentInitial,in Value,in Weight,in bPropagateToChildren)"}
   > &RigUnit_SetRelativeTransformForItem(
    Child=(Type=Bone,Name="belt_right_MSTR")
    Parent=(Type=Bone,Name="belt_right_MSTR")
    Value=(Rotation=(X=0.0349,Y=0,Z=0,W=0.9994),Translation=(X=0,Y=0,Z=-0.2),Scale3D=(X=1,Y=1,Z=1))
    Weight=1
    bParentInitial=False
    bPropagateToChildren=True
   ) {template="Set Relative Transform::Execute(in Child,in Parent,in bParentInitial,in Value,in Weight,in bPropagateToChildren)"}
   n8 = &RigVMFunction_AlphaInterp(
    InterpSpeedDecreasing=10
    InterpSpeedIncreasing=2
    Scale=1
    bClampResult=False
    bMapRange=false
    Value=&RigVMFunction_MathBoolToFloat(Value=RigVMVariableNode() {variable=jump or fall}.Value).Result
   )
   &RigUnit_SetRelativeTransformForItem(
    Child=(Type=Bone,Name="belt_back_MSTR")
    Parent=(Type=Bone,Name="belt_back_MSTR")
    Value=(Rotation=(X=0,Y=0.2588,Z=0,W=0.9659),Translation=(X=-8,Y=2,Z=1.2),Scale3D=(X=1,Y=1,Z=1))
    bParentInitial=False
    bPropagateToChildren=True
    Weight=@n8.Result
    ExecutePin=@RigVMFunction_Sequence_1.C
   ) {template="Set Relative Transform::Execute(in Child,in Parent,in bParentInitial,in Value,in Weight,in bPropagateToChildren)"}
   > &RigUnit_SetRelativeTransformForItem(
    Child=(Type=Bone,Name="belt_right_MSTR")
    Parent=(Type=Bone,Name="belt_right_MSTR")
    Value=(Rotation=(X=0,Y=0,Z=-0.5,W=0.866),Translation=(X=0,Y=8,Z=0),Scale3D=(X=1,Y=1,Z=1))
    bParentInitial=False
    bPropagateToChildren=True
    Weight=@n8.Result
   ) {template="Set Relative Transform::Execute(in Child,in Parent,in bParentInitial,in Value,in Weight,in bPropagateToChildren)"}
  ]}
]
hierarchy [
 root {loc=<0 0 0>, rot=<0 0 0 1>}
 pelvis {parent=root, loc=<0 -0.5301 98.0571>, rot=<0.0127 0.707 -0.0127 -0.707>}
 spine_01 {parent=pelvis, loc=<6.1208 -0.9328 0>, rot=<0 0 0.034 -0.9994>}
 spine_02 {parent=spine_01, loc=<6.5997 -0.2095 0>, rot=<0 0 0.0238 -0.9997>}
 spine_03 {parent=spine_02, loc=<6.4914 -0.1237 0>, rot=<0 0 0.0149 -0.9999>}
 spine_04 {parent=spine_03, loc=<6.9598 0.0794 0>, rot=<0 0 -0.0132 -0.9999>}
 spine_05 {parent=spine_04, loc=<6.9987 0 0>, rot=<0 0 -0.0015 -1>}
 neck_01 {parent=spine_05, loc=<11.3138 0.1529 0>, rot=<0 0 -0.2003 -0.9797>}
 neck_02 {parent=neck_01, loc=<4.5627 0 0>, rot=<0 0 0.0504 -0.9987>}
 head {parent=neck_02, loc=<5.3914 0 0>, rot=<0 0 0.115 -0.9934>}
 FACIAL_R_Jawline {parent=head, loc=<-2.9132 3.7089 3.8081>, rot=<0.6677 -0.4676 -0.4681 -0.3411>}
 FACIAL_L_Jawline {parent=head, loc=<-3.0708 3.6341 -4.1855>, rot=<-0.459 0.3563 0.6737 0.4567>}
 FACIAL_C_Jawline {parent=head, loc=<-5.9491 5.5475 -0.1441>, rot=<-0.6827 0.1638 0.6951 0.1546>}
 FACIAL_R_UnderChin {parent=head, loc=<-4.9438 2.3436 2.6227>, rot=<-0.3916 0.1097 0.8247 0.3932>}
 FACIAL_L_UnderChin {parent=head, loc=<-5.0507 2.2917 -2.8998>, rot=<0.8177 -0.3988 -0.4043 -0.0946>}
 FACIAL_C_UnderChin {parent=head, loc=<-6.1901 3.6273 -0.1226>, rot=<-0.6668 0.2232 0.6784 0.213>}
 dyn_bun_hair {parent=head, loc=<15.4363 -6.6038 -2.8967>, rot=<-0.0693 -0.8993 0.1078 0.4181>}
 dyn_right_front_hair {parent=head, loc=<4.742 3.7732 7.5893>, rot=<0.0046 -0.0099 0.0237 -0.9997>}
 MSTR_center_hair {parent=head, loc=<0.0349 -7.0463 -0.1541>, rot=<0.0092 -0.7098 -0.704 0.0235>}
 dyn_center_hair_01 {parent=MSTR_center_hair, loc=<0 0 0>, rot=<0 0 0 1>}
 dyn_center_hair_02 {parent=dyn_center_hair_01, loc=<5.385 0 0>, rot=<0 0 0 1>}
 MSTR_right_hair {parent=head, loc=<0.2398 -6.4003 4.5019>, rot=<-0.0255 -0.8702 -0.4921 -0.0032>}
 dyn_right_hair_01 {parent=MSTR_right_hair, loc=<0 0 0>, rot=<0 0 0 1>}
 dyn_right_hair_02 {parent=dyn_right_hair_01, loc=<4.224 0 0>, rot=<0 0 0 1>}
 MSTR_left_hair {parent=head, loc=<0.053 -6.4815 -4.825>, rot=<0.019 0.4278 0.9035 0.0173>}
 dyn_left_hair_01 {parent=MSTR_left_hair, loc=<0 0 0>, rot=<0 0 0 1>}
 dyn_left_hair_02 {parent=dyn_left_hair_01, loc=<4.2238 0 0>, rot=<0 0 0 1>}
 MSTR_bang {parent=head, loc=<10.1464 8.2952 -1.7546>, rot=<0.2462 0.6686 0.7015 0.0134>}
 dyn_bang_01 {parent=MSTR_bang, loc=<0 0 0>, rot=<0 0 0 1>}
 dyn_bang_02 {parent=dyn_bang_01, loc=<3.4813 0 0>, rot=<0 0 0 1>}
 MSTR_left_front_hair {parent=head, loc=<5.6205 3.4133 -7.2953>, rot=<-0.0754 -0.7075 -0.7 -0.062>}
 dyn_left_front_hair_01 {parent=MSTR_left_front_hair, loc=<0 0 0>, rot=<0 0 0 1>}
 dyn_left_front_hair_02 {parent=dyn_left_front_hair_01, loc=<3.7834 0 0>, rot=<0 0 0 1>}
 clavicle_l {parent=spine_05, loc=<5.4831 -0.1053 -3.2946>, rot=<0.0465 -0.7289 -0.009 -0.6829>}
 upperarm_l {parent=clavicle_l, loc=<9.8658 -1.2136 -0.4105>, rot=<0.0164 -0.3245 -0.0269 -0.9454>}
 lowerarm_l {parent=upperarm_l, loc=<26.465 -0.8345 0.259>, rot=<0.0036 0.0161 -0.2298 -0.9731>}
 hand_l {parent=lowerarm_l, loc=<22.7355 -0.3203 -0.6105>, rot=<-0.6784 -0.0223 0.0735 -0.7307>}
 thumb_01_l {parent=hand_l, loc=<1.3118 -0.3191 -2.459>, rot=<0.6546 -0.4042 0.0664 -0.6354>}
 thumb_02_l {parent=thumb_01_l, loc=<5.9336 -0.0448 0.0401>, rot=<-0.0037 0.0213 0.298 -0.9543>}
 thumb_03_l {parent=thumb_02_l, loc=<2.2223 0.3682 -0.0907>, rot=<0.0118 -0.013 0.0066 -0.9998>}
 index_metacarpal_l {parent=hand_l, loc=<1.6535 0.0183 -2.96>, rot=<0.0158 -0.0124 -0.0173 -0.9996>}
 index_01_l {parent=index_metacarpal_l, loc=<8.3008 0.0357 -0.0524>, rot=<-0.0044 0.0485 0.0924 -0.9945>}
 index_02_l {parent=index_01_l, loc=<3.3587 -0.3004 0.0452>, rot=<-0.002 0.0027 0.3225 -0.9466>}
 index_03_l {parent=index_02_l, loc=<2.5578 -0.123 -0.0252>, rot=<-0.002 0.0067 0.2558 -0.9667>}
 middle_metacarpal_l {parent=hand_l, loc=<2.1856 -0.1725 -0.5984>, rot=<0.006 0.0168 -0.0329 -0.9993>}
 middle_01_l {parent=middle_metacarpal_l, loc=<7.3275 -0.2447 0.0129>, rot=<-0.0208 0.0042 0.0858 -0.9961>}
 middle_02_l {parent=middle_01_l, loc=<4.0123 0.1401 0.1175>, rot=<-0.0218 0.0258 0.3512 -0.9357>}
 middle_03_l {parent=middle_02_l, loc=<2.6077 0.0534 0.1373>, rot=<0.0082 0.0188 0.1708 -0.9851>}
 ring_metacarpal_l {parent=hand_l, loc=<2.2074 -0.3625 1.3638>, rot=<-0.0284 0.0722 -0.0458 -0.9959>}
 ring_01_l {parent=ring_metacarpal_l, loc=<7.0295 -0.6197 -0.1446>, rot=<-0.0489 -0.0252 0.1892 -0.9804>}
 ring_02_l {parent=ring_01_l, loc=<3.6174 0.4647 -0.1027>, rot=<0 0 0.2806 -0.9598>}
 ring_03_l {parent=ring_02_l, loc=<2.4152 0.114 0.0199>, rot=<0 0 0.1189 -0.9929>}
 pinky_metacarpal_l {parent=hand_l, loc=<1.7616 -0.9753 3.3814>, rot=<-0.0397 0.0814 -0.069 -0.9935>}
 pinky_01_l {parent=pinky_metacarpal_l, loc=<7.1709 -0.4831 -0.1094>, rot=<-0.0654 -0.0035 0.2242 -0.9723>}
 pinky_02_l {parent=pinky_01_l, loc=<2.8274 0.0893 -0.1043>, rot=<0.0037 0.0011 0.2854 -0.9584>}
 pinky_03_l {parent=pinky_02_l, loc=<2.2046 0.0307 -0.0078>, rot=<-0.0002 0.0098 0.1445 -0.9895>}
 deform_wrist_dn_base_l {parent=hand_l, loc=<0 -1.5326 0>, rot=<0 0 0 -1>}
 deform_wrist_dn_l {parent=deform_wrist_dn_base_l, loc=<0 0 0>, rot=<0 0 0 1>}
 deform_wrist_up_base_l {parent=hand_l, loc=<0 2.3009 0>, rot=<0 0 0 -1>}
 deform_wrist_up_l {parent=deform_wrist_up_base_l, loc=<0 0 0>, rot=<0 0 0 1>}
 hand_attach_l {parent=hand_l, loc=<8.5 -3 0>, rot=<0.7071 0.7071 0 0>}
 weapon_l {parent=hand_l, loc=<-0.1974 -3.4111 0.3811>, rot=<0 0 0 1>}
 lowerarm_twist_02_l {parent=lowerarm_l, loc=<10.6707 0 0>, rot=<0 0 0 1>}
 dyn_braclet_01 {parent=lowerarm_twist_02_l, loc=<4.3201 -0.291 -0.7547>, rot=<0.0742 0.3327 0.9379 -0.0637>}
 dyn_braclet_02 {parent=lowerarm_twist_02_l, loc=<5.5725 -0.3356 -0.8503>, rot=<0.0451 0.3379 0.9399 0.0178>}
 dyn_braclet_03 {parent=lowerarm_twist_02_l, loc=<6.7392 -0.3532 -0.9366>, rot=<-0.0399 0.3427 0.9381 0.0301>}
 lowerarm_twist_01_l {parent=lowerarm_l, loc=<18.316 0 0>, rot=<0 0 0 1>}
 deform_elbow_in_base_l {parent=lowerarm_l, loc=<0.1834 3.6062 -0.2572>, rot=<0 0 0 -1>}
 deform_elbow_in_l {parent=deform_elbow_in_base_l, loc=<0 0 0>, rot=<0 0 0 1>}
 deform_elbow_out_base_l {parent=lowerarm_l, loc=<0 -4 0>, rot=<0 0 0 1>}
 deform_elbow_out_l {parent=deform_elbow_out_base_l, loc=<0 0 0>, rot=<0 0 0 1>}
 upperarm_twist_01_l {parent=upperarm_l, loc=<7.2135 -0.1134 0>, rot=<0 0 0 1>}
 deform_upperarm_bk_base_l {parent=upperarm_twist_01_l, loc=<-6.4053 -4.3746 0.314>, rot=<0 0 0 -1>}
 deform_upperarm_bk_l {parent=deform_upperarm_bk_base_l, loc=<0 0 0>, rot=<0 0 0 1>}
 deform_upperarm_fr_base_l {parent=upperarm_twist_01_l, loc=<-5.6312 5.092 0.3354>, rot=<0 0 0 -1>}
 deform_upperarm_fr_l {parent=deform_upperarm_fr_base_l, loc=<0 0 0>, rot=<0 0 0 1>}
 deform_upperarm_up_base_l {parent=upperarm_twist_01_l, loc=<-8.891 0.4825 3.9517>, rot=<0 0 0 -1>}
 deform_upperarm_up_l {parent=deform_upperarm_up_base_l, loc=<0 0 0>, rot=<0 0 0 1>}
 upperarm_twist_02_l {parent=upperarm_l, loc=<19.7998 -0.1134 0>, rot=<0 0 0 1>}
 clavicle_r {parent=spine_05, loc=<5.4831 -0.1053 3.2946>, rot=<0.7289 0.0465 -0.6829 0.009>}
 upperarm_r {parent=clavicle_r, loc=<-9.8658 1.2136 0.4105>, rot=<0.0164 -0.3245 -0.0269 -0.9454>}
 lowerarm_r {parent=upperarm_r, loc=<-26.465 0.8345 -0.259>, rot=<0.0036 0.0161 -0.2298 -0.9731>}
 hand_r {parent=lowerarm_r, loc=<-22.7355 0.3203 0.6106>, rot=<-0.6784 -0.0223 0.0735 -0.7307>}
 thumb_01_r {parent=hand_r, loc=<-1.3118 0.3191 2.459>, rot=<0.6546 -0.4042 0.0664 -0.6354>}
 thumb_02_r {parent=thumb_01_r, loc=<-5.9336 0.0448 -0.0401>, rot=<-0.0037 0.0213 0.298 -0.9543>}
 thumb_03_r {parent=thumb_02_r, loc=<-2.2223 -0.3682 0.0907>, rot=<0.0118 -0.013 0.0066 -0.9998>}
 index_metacarpal_r {parent=hand_r, loc=<-1.6535 -0.0183 2.96>, rot=<0.0158 -0.0124 -0.0173 -0.9996>}
 index_01_r {parent=index_metacarpal_r, loc=<-8.3008 -0.0357 0.0524>, rot=<-0.0044 0.0485 0.0924 -0.9945>}
 index_02_r {parent=index_01_r, loc=<-3.3587 0.3004 -0.0452>, rot=<-0.002 0.0027 0.3225 -0.9466>}
 index_03_r {parent=index_02_r, loc=<-2.5578 0.123 0.0252>, rot=<-0.002 0.0067 0.2558 -0.9667>}
 middle_metacarpal_r {parent=hand_r, loc=<-2.1856 0.1725 0.5984>, rot=<0.006 0.0168 -0.0329 -0.9993>}
 middle_01_r {parent=middle_metacarpal_r, loc=<-7.3275 0.2447 -0.0129>, rot=<-0.0208 0.0042 0.0858 -0.9961>}
 middle_02_r {parent=middle_01_r, loc=<-4.0123 -0.1401 -0.1175>, rot=<-0.0218 0.0258 0.3512 -0.9357>}
 middle_03_r {parent=middle_02_r, loc=<-2.6077 -0.0534 -0.1373>, rot=<0.0082 0.0188 0.1708 -0.9851>}
 ring_metacarpal_r {parent=hand_r, loc=<-2.2074 0.3625 -1.3638>, rot=<-0.0284 0.0722 -0.0458 -0.9959>}
 ring_01_r {parent=ring_metacarpal_r, loc=<-7.0295 0.6197 0.1446>, rot=<-0.0489 -0.0252 0.1892 -0.9804>}
 ring_02_r {parent=ring_01_r, loc=<-3.6174 -0.4647 0.1027>, rot=<0 0 0.2806 -0.9598>}
 ring_03_r {parent=ring_02_r, loc=<-2.4152 -0.114 -0.0199>, rot=<0 0 0.1189 -0.9929>}
 pinky_metacarpal_r {parent=hand_r, loc=<-1.7616 0.9753 -3.3814>, rot=<-0.0397 0.0814 -0.069 -0.9935>}
 pinky_01_r {parent=pinky_metacarpal_r, loc=<-7.1709 0.4831 0.1094>, rot=<-0.0654 -0.0035 0.2242 -0.9723>}
 pinky_02_r {parent=pinky_01_r, loc=<-2.8274 -0.0893 0.1043>, rot=<0.0037 0.0011 0.2854 -0.9584>}
 pinky_03_r {parent=pinky_02_r, loc=<-2.2046 -0.0307 0.0078>, rot=<-0.0002 0.0098 0.1445 -0.9895>}
 deform_wrist_dn_base_r {parent=hand_r, loc=<-0.0001 1.5328 0>, rot=<0 0 0 -1>}
 deform_wrist_dn_r {parent=deform_wrist_dn_base_r, loc=<0 0 0>, rot=<0 0 0 1>}
 deform_wrist_up_base_r {parent=hand_r, loc=<-0.0002 -2.3007 -0.0001>, rot=<0 0 0 -1>}
 deform_wrist_up_r {parent=deform_wrist_up_base_r, loc=<0 0 0>, rot=<0 0 0 1>}
 weapon_r {parent=hand_r, loc=<0.1974 3.4111 -0.3811>, rot=<0 0 0 1>}
 hand_attach_r {parent=hand_r, loc=<-8.5001 2.9999 0>, rot=<0 0 -0.7071 -0.7071>}
 lowerarm_twist_02_r {parent=lowerarm_r, loc=<-10.6707 0 0>, rot=<0 0 0 -1>}
 lowerarm_twist_01_r {parent=lowerarm_r, loc=<-18.316 0 0>, rot=<0 0 0 -1>}
 deform_elbow_in_base_r {parent=lowerarm_r, loc=<-0.1834 -3.6062 0.2572>, rot=<0 0 0 -1>}
 deform_elbow_in_r {parent=deform_elbow_in_base_r, loc=<0 0 0>, rot=<0 0 0 1>}
 deform_elbow_out_base_r {parent=lowerarm_r, loc=<0 4 -0.0001>, rot=<0 0 0 -1>}
 deform_elbow_out_r {parent=deform_elbow_out_base_r, loc=<0 0 0>, rot=<0 0 0 1>}
 upperarm_twist_01_r {parent=upperarm_r, loc=<-7.2135 0.1134 0>, rot=<0 0 0 -1>}
 deform_upperarm_bk_base_r {parent=upperarm_twist_01_r, loc=<6.4053 4.3746 -0.314>, rot=<0 0 0 -1>}
 deform_upperarm_bk_r {parent=deform_upperarm_bk_base_r, loc=<0 0 0>, rot=<0 0 0 1>}
 deform_upperarm_fr_base_r {parent=upperarm_twist_01_r, loc=<5.6312 -5.092 -0.3354>, rot=<0 0 0 -1>}
 deform_upperarm_fr_r {parent=deform_upperarm_fr_base_r, loc=<0 0 0>, rot=<0 0 0 1>}
 deform_upperarm_up_base_r {parent=upperarm_twist_01_r, loc=<8.891 -0.4825 -3.9517>, rot=<0 0 0 -1>}
 deform_upperarm_up_r {parent=deform_upperarm_up_base_r, loc=<0 0 0>, rot=<0 0 0 1>}
 upperarm_twist_02_r {parent=upperarm_r, loc=<-19.7998 0.1134 0>, rot=<0 0 0 -1>}
 attach_backpack {parent=spine_05, loc=<-0.123 -13.473 0>, rot=<-0.0093 -0.707 0.0093 -0.707>}
 Backpack_BR {parent=attach_backpack, type=Null, loc=<0 -3.3294 0.3556>, rot=<0 0 0 1>}
 attach_cape {parent=spine_05, loc=<9.017 -11.714 0>, rot=<0.0616 -0.7044 -0.0616 -0.7044>}
 cape {parent=attach_cape, type=Null, loc=<0 -2.9139 -0.2756>, rot=<0 0 0 1>}
 dyn_pendant_03 {parent=spine_05, loc=<4.2684 7.3333 -0.0034>, rot=<-0.2592 -0.6579 -0.6573 -0.2606>}
 dyn_pendant_04 {parent=dyn_pendant_03, loc=<3.0666 -0.0019 0.0479>, rot=<0 0 0 1>}
 dyn_pendant_05 {parent=dyn_pendant_04, loc=<3.0536 0.0103 -0.2567>, rot=<0 0 0 1>}
 base_boa {parent=spine_05, loc=<-4.6122 -12.0338 0>, rot=<0.0284 -0.7065 -0.0284 -0.7065>}
 boa_root {parent=base_boa, loc=<0 0 0>, rot=<0 0 0 1>}
 def_boa_left_01 {parent=boa_root, loc=<7.3961 0.0489 3.6894>, rot=<-0.5519 0.2463 -0.3247 -0.7276>}
 def_boa_left_02 {parent=def_boa_left_01, loc=<8.2579 0 0>, rot=<0.1873 -0.2853 -0.1407 -0.9294>}
 def_boa_left_03 {parent=def_boa_left_02, loc=<7.7252 0 0>, rot=<0.1236 -0.3905 -0.113 -0.9052>}
 def_boa_left_04 {parent=def_boa_left_03, loc=<8.2966 0.1064 -0.3544>, rot=<0.0668 -0.5051 -0.0437 -0.8593>}
 dyn_boa_left_05_MSTR {parent=def_boa_left_04, loc=<14.7834 0 0>, rot=<0 -0.1446 0 -0.9895>}
 dyn_boa_left_05 {parent=dyn_boa_left_05_MSTR, loc=<0 0 0>, rot=<0 0 0 1>}
 dyn_boa_left_06_MSTR {parent=dyn_boa_left_05, loc=<13.0048 0 0>, rot=<0 0 0 1>}
 dyn_boa_left_06 {parent=dyn_boa_left_06_MSTR, loc=<0 0 0>, rot=<0 0 0 1>}
 def_boa_right_01 {parent=boa_root, loc=<-7.3387 0.0407 3.7762>, rot=<0.2455 -0.5512 0.7285 0.3244>}
 def_boa_right_02 {parent=def_boa_right_01, loc=<8.2461 0 0>, rot=<-0.1908 -0.2874 0.1373 -0.9285>}
 def_boa_right_03 {parent=def_boa_right_02, loc=<7.7195 0 0>, rot=<-0.1247 -0.3862 0.1158 -0.9066>}
 def_boa_right_04 {parent=def_boa_right_03, loc=<8.2821 -0.1101 -0.35>, rot=<-0.0667 -0.5039 0.0461 -0.8599>}
 dyn_boa_right_05_MSTR {parent=def_boa_right_04, loc=<14.7544 0 0>, rot=<0 -0.1446 0 -0.9895>}
 dyn_boa_right_05 {parent=dyn_boa_right_05_MSTR, loc=<0 0 0>, rot=<0 0 0 1>}
 dyn_boa_right_06_MSTR {parent=dyn_boa_right_05, loc=<13.0048 0 0>, rot=<0 0 0 1>}
 dyn_boa_right_06 {parent=dyn_boa_right_06_MSTR, loc=<0 0 0>, rot=<0 0 0 1>}
 dyn_back_boa {parent=spine_05, loc=<11.9159 -6.5434 -0.0897>, rot=<0.0401 -0.9992 0 0>}
 deform_pec_aim_base_l {parent=spine_04, loc=<3.2456 9.3309 -1.2237>, rot=<0.0603 -0.4033 0.3318 -0.8507>}
 deform_pec_aim_l {parent=deform_pec_aim_base_l, loc=<0.0022 0.0489 -0.0335>, rot=<0.0014 0.0009 0.0013 -1>}
 deform_pec_base_l {parent=deform_pec_aim_l, loc=<9.3644 -0.0239 0.0163>, rot=<-0.0014 -0.0009 -0.0013 -1>}
 deform_pec_l {parent=deform_pec_base_l, loc=<0 0 0>, rot=<0 0 0 1>}
 deform_pec_aim_base_r {parent=spine_04, loc=<3.2452 9.3309 1.2237>, rot=<-0.4033 -0.0603 0.8507 0.3318>}
 deform_pec_aim_r {parent=deform_pec_aim_base_r, loc=<0 0 0>, rot=<0 0 0 1>}
 deform_pec_base_r {parent=deform_pec_aim_r, loc=<-9.3667 -0.0003 -0.0004>, rot=<0 0 0 1>}
 deform_pec_r {parent=deform_pec_base_r, loc=<0 0 0>, rot=<0 0 0 1>}
 MSTR_right_bag {parent=spine_03, loc=<0.0403 8.594 10.0607>, rot=<0.0583 -0.9045 -0.4217 0.0251>}
 MSTR_left_bag {parent=spine_03, loc=<0.0863 8.6717 -10.0401>, rot=<0.0602 -0.9021 0.4245 -0.0501>}
 thigh_l {parent=pelvis, loc=<-1.0855 -0.2109 -8.7053>, rot=<0.0009 0.0355 0.0155 -0.9993>}
 calf_l {parent=thigh_l, loc=<-42.5284 -0.0019 -0.7297>, rot=<0.0006 -0.0161 -0.0391 -0.9991>}
 foot_l {parent=calf_l, loc=<-46.7401 -2.5014 -0.592>, rot=<0.0427 -0.0238 0.0353 -0.9982>}
 ball_l {parent=foot_l, loc=<-3.5658 10.8066 -0.1544>, rot=<0.0001 -0.0001 -0.7063 -0.7079>}
 calf_twist_02_l {parent=calf_l, loc=<-19.1546 -1.7943 -0.2638>, rot=<0.0237 0.0095 -0.0035 -0.9997>}
 calf_twist_01_l {parent=calf_l, loc=<-35.8483 -1.9197 -0.5794>, rot=<0.0237 0.0095 -0.0035 -0.9997>}
 deform_knee_in_base_l {parent=calf_l, loc=<0.8031 -5.0231 0.2456>, rot=<0 0 0 -1>}
 deform_knee_in_l {parent=deform_knee_in_base_l, loc=<0 0 0>, rot=<0 0 0 1>}
 deform_knee_out_base_l {parent=calf_l, loc=<0 4 0>, rot=<0 0 0 -1>}
 deform_knee_out_l {parent=deform_knee_out_base_l, loc=<0 0 0>, rot=<0 0 0 1>}
 thigh_twist_01_l {parent=thigh_l, loc=<-27.952 0 0>, rot=<0.0013 0 0 -1>}
 deform_groin_base_l {parent=thigh_l, loc=<0.4801 5.6776 -0.2486>, rot=<0 0 0 -1>}
 deform_groin_l {parent=deform_groin_base_l, loc=<0 0 0>, rot=<0 0 0 1>}
 MSTR_waist_bag {parent=thigh_l, loc=<-9.325 0.1691 -9.4759>, rot=<-0.0024 -0.999 -0.0001 -0.0435>}
 dyn_waist_bag {parent=MSTR_waist_bag, loc=<0 0 0>, rot=<0 0 0 1>}
 dyn_zip {parent=dyn_waist_bag, loc=<0.391 9.6479 1.049>, rot=<-0.8516 -0.0652 -0.0307 -0.5191>}
 bag_belt_front_def {parent=thigh_l, loc=<-6.056 8.8013 1.7703>, rot=<-0.0024 -0.9994 -0.0003 0.0355>}
 thigh_r {parent=pelvis, loc=<-1.0855 -0.2109 8.7053>, rot=<0.0355 -0.0009 0.9993 0.0155>}
 calf_r {parent=thigh_r, loc=<42.5284 0.0019 0.7297>, rot=<0.0006 -0.0161 -0.0391 -0.9991>}
 foot_r {parent=calf_r, loc=<46.7401 2.5014 0.592>, rot=<0.0427 -0.0238 0.0353 -0.9982>}
 ball_r {parent=foot_r, loc=<3.5658 -10.8066 0.1544>, rot=<0.0001 -0.0001 -0.7063 -0.7079>}
 calf_twist_02_r {parent=calf_r, loc=<19.1546 1.7943 0.2638>, rot=<0.0237 0.0095 -0.0035 -0.9997>}
 calf_twist_01_r {parent=calf_r, loc=<35.8483 1.9197 0.5794>, rot=<0.0237 0.0095 -0.0035 -0.9997>}
 deform_knee_in_base_r {parent=calf_r, loc=<-0.8031 5.0231 -0.2457>, rot=<0 0 0 -1>}
 deform_knee_in_r {parent=deform_knee_in_base_r, loc=<0 0 0>, rot=<0 0 0 1>}
 deform_knee_out_base_r {parent=calf_r, loc=<0 -4 0>, rot=<0 0 0 -1>}
 deform_knee_out_r {parent=deform_knee_out_base_r, loc=<0 0 0>, rot=<0 0 0 1>}
 thigh_twist_01_r {parent=thigh_r, loc=<27.952 0 0>, rot=<0.0013 0 0 -1>}
 deform_groin_base_r {parent=thigh_r, loc=<-0.4801 -5.6776 0.2486>, rot=<0 0 0 -1>}
 deform_groin_r {parent=deform_groin_base_r, loc=<0 0 0>, rot=<0 0 0 1>}
 deform_glute_base_l {parent=pelvis, loc=<-7.3466 -9.0213 -9.1245>, rot=<0.0172 0.031 -0.4477 -0.8935>}
 deform_glute_l {parent=deform_glute_base_l, loc=<0 0 0>, rot=<0 0 0 1>}
 deform_glute_base_r {parent=pelvis, loc=<-7.3464 -9.0213 9.1245>, rot=<0.031 -0.0172 0.8935 -0.4477>}
 deform_glute_r {parent=deform_glute_base_r, loc=<0 0 0>, rot=<0 0 0 1>}
 bag_belt_def {parent=pelvis, loc=<-6.3867 -12.9471 -7.0634>, rot=<-0.0127 -0.707 0.0127 -0.707>}
 bag_belt_def_back {parent=pelvis, loc=<-2.5501 -12.9029 2.6814>, rot=<-0.0127 -0.707 0.0127 -0.707>}
 belt_right_MSTR {parent=pelvis, loc=<-3.6604 -0.2174 17.1658>, rot=<-0.0179 -0.9995 0.0003 -0.0267>}
 bag_belt_def_right {parent=pelvis, loc=<1.8795 -1.4944 14.689>, rot=<-0.0155 -0.8635 0.009 -0.5041>}
 belt_back_MSTR {parent=pelvis, loc=<-0.9825 -12.5662 7.7789>, rot=<-0.0179 -0.9998 0 0>}
 MSTR_right_front_belts {parent=pelvis, loc=<-1.8577 7.9934 8.4047>, rot=<-0.0127 -0.707 0.0127 -0.707>}
 attach {parent=root, loc=<0 0 0>, rot=<0 0 0 1>}
 ik_hand_root {parent=root, loc=<0 0 0>, rot=<0 0 0 1>}
 ik_hand_gun {parent=ik_hand_root, loc=<-48.1743 3.3545 104.4381>, rot=<0.6276 0.1309 0.4183 -0.6434>}
 ik_hand_r {parent=ik_hand_gun, loc=<0 0 0>, rot=<0 0 0 1>}
 ik_hand_l {parent=ik_hand_gun, loc=<59.3271 67.6997 34.3524>, rot=<-0.6158 -0.7027 -0.3565 0>}
 ik_foot_root {parent=root, loc=<0 0 0>, rot=<0 0 0 1>}
 ik_foot_l {parent=ik_foot_root, loc=<14.8262 -7.3545 8.296>, rot=<0.0357 0.7045 0.0284 -0.7082>}
 ik_foot_r {parent=ik_foot_root, loc=<-14.8262 -7.3545 8.296>, rot=<0.7082 0.0284 -0.7045 0.0357>}
 root_fp {parent=root, loc=<0 0 0>, rot=<0 0 0 1>}
 attach_fp {parent=root_fp, loc=<0 0 0>, rot=<0 0 0 1>}
 camera_root {parent=root_fp, loc=<0 0 0>, rot=<0 0 0 1>}
 camera {parent=camera_root, loc=<0 0 0>, rot=<0 0 0 1>}
 attach_emote_01 {parent=root, loc=<0 0 0>, rot=<0 0 0 1>}
 attach_emote_02 {parent=root, loc=<0 0 0>, rot=<0 0 0 1>}
 attach_emote_03 {parent=root, loc=<0 0 0>, rot=<0 0 0 1>}
 warp_point_01 {parent=root, loc=<0 0 0>, rot=<0 0 0 1>}
 warp_point_02 {parent=root, loc=<0 0 0>, rot=<0 0 0 1>}
 warp_point_03 {parent=root, loc=<0 0 0>, rot=<0 0 0 1>}
 def_boa_left_01_null {type=Null, loc=<7.3961 -15.1461 129.2867>, rot=<-0.5519 0.2463 -0.3247 -0.7276>}
 def_boa_left_01_ctrl {controlType=EULER_TRANSFORM, parent=def_boa_left_01_null, type=Control, loc=<0 0 0>, rot=<0 0 0 1>}
 def_boa_left_02_null {type=Null, loc=<12.9109 -13.4896 135.2058>, rot=<0.2493 -0.1598 0.5155 0.8041>}
 def_boa_left_02_ctrl {controlType=EULER_TRANSFORM, parent=def_boa_left_02_null, type=Control, loc=<0 0 0>, rot=<0 0 0 1>}
 def_boa_left_03_null {type=Null, loc=<16.1362 -7.7013 139.1772>, rot=<0.093 -0.0775 -0.6351 -0.7629>}
 def_boa_left_03_ctrl {controlType=EULER_TRANSFORM, parent=def_boa_left_03_null, type=Control, loc=<0 0 0>, rot=<0 0 0 1>}
 def_boa_left_04_null {type=Null, loc=<17.5353 0.1522 136.8674>, rot=<-0.4483 0.4136 0.5373 0.5825>}
 def_boa_left_04_ctrl {controlType=EULER_TRANSFORM, parent=def_boa_left_04_null, type=Control, loc=<0 0 0>, rot=<0 0 0 1>}
 def_boa_right_01_null {type=Null, loc=<-7.3388 -15.1543 129.3735>, rot=<0.2455 -0.5512 0.7285 0.3244>}
 def_boa_right_01_ctrl {controlType=EULER_TRANSFORM, parent=def_boa_right_01_null, type=Control, loc=<0 0 0>, rot=<0 0 0 1>}
 def_boa_right_02_null {type=Null, loc=<-12.8552 -13.4885 135.2719>, rot=<-0.1562 0.2459 -0.8076 -0.5128>}
 def_boa_right_02_ctrl {controlType=EULER_TRANSFORM, parent=def_boa_right_02_null, type=Control, loc=<0 0 0>, rot=<0 0 0 1>}
 def_boa_right_03_null {type=Null, loc=<-16.1383 -7.6877 139.1658>, rot=<-0.0779 0.0939 0.7637 0.6339>}
 def_boa_right_03_ctrl {controlType=EULER_TRANSFORM, parent=def_boa_right_03_null, type=Control, loc=<0 0 0>, rot=<0 0 0 1>}
 def_boa_right_04_null {type=Null, loc=<-17.5546 0.1454 136.8498>, rot=<0.4139 -0.4475 -0.582 -0.5382>}
 def_boa_right_04_ctrl {controlType=EULER_TRANSFORM, parent=def_boa_right_04_null, type=Control, loc=<0 0 0>, rot=<0 0 0 1>}
]
variables [
 DBNO {type=bool, default=False}
 Parachute opened {type=bool, default=False}
 RBAN {type=bool, default=False}
 Skidive floating {type=bool, default=False}
 boa_left_ctrl {
  type="TArray<FRigElementKey>"
  default=((Type=Control,Name="def_boa_left_01_ctrl"),(Type=Control,Name="def_boa_left_02_ctrl"),(Type=Control,Name="def_boa_left_03_ctrl"),(Type=Control,Name="def_boa_left_04_ctrl"))
 }
 boa_left_joints {
  type="TArray<FRigElementKey>"
  default=((Type=Bone,Name="def_boa_left_01"),(Type=Bone,Name="def_boa_left_02"),(Type=Bone,Name="def_boa_left_03"),(Type=Bone,Name="def_boa_left_03"))
 }
 boa_right_ctrl {
  type="TArray<FRigElementKey>"
  default=((Type=Control,Name="def_boa_right_01_ctrl"),(Type=Control,Name="def_boa_right_02_ctrl"),(Type=Control,Name="def_boa_right_03_ctrl"),(Type=Control,Name="def_boa_right_04_ctrl"))
 }
 boa_right_joints {
  type="TArray<FRigElementKey>"
  default=((Type=Bone,Name="def_boa_right_01"),(Type=Bone,Name="def_boa_right_02"),(Type=Bone,Name="def_boa_right_03"),(Type=Bone,Name="def_boa_right_04"))
 }
 crouching {type=bool, default=False}
 jump or fall {type=bool, default=False}
 playing emote {type=bool, default=False}
 targeting {type=bool, default=False}
]
