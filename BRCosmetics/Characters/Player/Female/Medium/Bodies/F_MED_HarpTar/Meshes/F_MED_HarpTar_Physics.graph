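# Physics asset for the HarpTar medium female player body, covering a two-sided boa and a waist bag.
# Capsules on pelvis, spine, thighs and arms form a partial body alongside the boa and bag bodies.
# Boa chains hang from spine_04 with limited swing and tight twist, except the free right root.
# Most limb constraints are locked, while the bag-to-left-thigh and right-upper-arm-to-spine constraints are fully free.
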PhysicsAsset /BRCosmetics/Characters/Player/Female/Medium/Bodies/F_MED_HarpTar/Meshes/F_MED_HarpTar_Physics
bodies {
 def_boa_left_04 [{shapeType=Capsule, length=10.5097, radius=4.3456, shapeName=def_boa_left_04_capsule, center=<5.6235 0 0>, rotation=<rot 0 -90.7989 -85.5068>}]
 def_boa_right_03 [{
   shapeType=Capsule
   length=4.9027
   radius=4.8557
   shapeName=def_boa_right_03_capsule
   center=<5.178 1.2673 0.6297>
   rotation=<rot 15.2741 -89.4993 -90>
  }]
 def_boa_right_04 [{
   shapeType=Capsule
   length=8.5251
   radius=3.6186
   shapeName=def_boa_right_04_capsule
   center=<5.0114 0.6105 1.1612>
   rotation=<rot 0 -93.1066 -90.3301>
  }]
 dyn_boa_left_05 [{
   shapeType=Capsule
   length=6.7461
   radius=3.7591
   shapeName=dyn_boa_left_05_capsule
   center=<6.4507 1.1507 -0.3324>
   rotation=<rot -4.4696 -86.5118 -83.8023>
  }]
 dyn_boa_left_06 [{
   shapeType=Capsule
   length=11.1088
   radius=3.1327
   shapeName=dyn_boa_left_06_capsule
   center=<8.9064 1.3617 -1.3034>
   rotation=<rot 0.0161 87.3933 -89.5939>
  }]
 dyn_boa_right_05 [{
   shapeType=Capsule
   length=9.4482
   radius=3.4372
   shapeName=dyn_boa_right_05_capsule
   center=<4.7759 -1.1473 -0.4069>
   rotation=<rot 20.4485 -95.5959 -88.9076>
  }]
 dyn_boa_right_06 [{
   shapeType=Capsule
   length=12.1309
   radius=3.3342
   shapeName=dyn_boa_right_06_capsule
   center=<7.69 -1.1294 -1.5214>
   rotation=<rot -0.0088 95.6584 -89.9547>
  }]
 dyn_waist_bag [{
   shapeType=Capsule
   length=9.3734
   radius=7.2946
   shapeName=dyn_waist_bag_capsule
   center=<4.4571 0.0235 -0.1871>
   rotation=<rot 0 2.889 -88.1613>
  }]
 lowerarm_l [{
   shapeType=Capsule
   length=22.4557
   radius=2.9205
   shapeName=lowerarm_l_capsule
   center=<17.7404 -0.2103 -0.1432>
   rotation=<rot 89.0047 -0.003 -0.003>
  }]
 lowerarm_r [{
   shapeType=Capsule
   length=21.9693
   radius=2.3577
   shapeName=lowerarm_r_capsule
   center=<-13.2543 0.2103 0.0653>
   rotation=<rot 89.0048 -0.0066 -177.9603>
  }]
 pelvis [{shapeType=Capsule, length=7.0043, radius=11.1367, shapeName=pelvis_capsule, center=<0 -1.5793 0>}]
 spine_02 [
  {
   shapeType=Capsule
   length=8.6442
   radius=5.685
   shapeName=spine_02_capsule
   center=<0.9414 1.9411 -0.0231>
   rotation=<rot -0.0022 -90.0112 179.2769>
  }
  {shapeType=Capsule, length=9.4095, radius=5.9557, shapeName=spine_02_capsule, center=<6.1321 0 0>}
 ]
 spine_04 [{shapeType=Capsule, length=4.2023, radius=9.237, shapeName=spine_04_capsule, center=<0.3551 2.9685 0>, rotation=<rot 0 12.6024 0>}]
 thigh_l [{shapeType=Capsule, length=25.7892, radius=7.9987, shapeName=thigh_l_capsule, center=<-21.8671 0 0>, rotation=<rot 90 0 0>}]
 thigh_r [{shapeType=Capsule, length=25.7892, radius=7.9987, shapeName=thigh_r_capsule, center=<21.8671 0 0>, rotation=<rot 90 180 0>}]
 upperarm_r [{
   shapeType=Capsule
   length=16.2375
   radius=2.9328
   shapeName=upperarm_r_capsule
   center=<-14.7457 0.6785 0.4462>
   rotation=<rot 0.4243 88.7933 -85.7908>
  }]
}
constraints [
 {
  bone1Name=lowerarm_l
  bone2Name=spine_04
  swing1LimitDegrees=45
  swing1Motion=Locked
  swing2LimitDegrees=45
  swing2Motion=Locked
  twistLimitDegrees=45
  twistMotion=Locked
 }
 {
  bone1Name=thigh_l
  bone2Name=pelvis
  swing1LimitDegrees=45
  swing1Motion=Locked
  swing2LimitDegrees=45
  swing2Motion=Locked
  twistLimitDegrees=45
  twistMotion=Locked
 }
 {
  bone1Name=thigh_r
  bone2Name=pelvis
  swing1LimitDegrees=45
  swing1Motion=Locked
  swing2LimitDegrees=45
  swing2Motion=Locked
  twistLimitDegrees=45
  twistMotion=Locked
 }
 {
  bone1Name=lowerarm_r
  bone2Name=upperarm_r
  swing1LimitDegrees=45
  swing1Motion=Locked
  swing2LimitDegrees=45
  swing2Motion=Locked
  twistLimitDegrees=45
  twistMotion=Locked
 }
 {bone1Name=dyn_boa_right_06, bone2Name=dyn_boa_right_05, swing1LimitDegrees=20, swing2LimitDegrees=20, twistLimitDegrees=2}
 {bone1Name=dyn_boa_right_05, bone2Name=def_boa_right_04, swing1LimitDegrees=30, swing2LimitDegrees=30, twistLimitDegrees=2}
 {bone1Name=dyn_boa_left_06, bone2Name=dyn_boa_left_05, swing1LimitDegrees=35, swing2LimitDegrees=35, twistLimitDegrees=2}
 {bone1Name=dyn_boa_left_05, bone2Name=def_boa_left_04, swing1LimitDegrees=35, swing2LimitDegrees=35, twistLimitDegrees=2}
 {bone1Name=def_boa_right_04, bone2Name=def_boa_right_03, swing1LimitDegrees=25, swing2LimitDegrees=25, twistLimitDegrees=4}
 {bone1Name=def_boa_left_04, bone2Name=spine_04, swing1LimitDegrees=15, swing2LimitDegrees=20, twistLimitDegrees=5}
 {
  bone1Name=dyn_waist_bag
  bone2Name=thigh_l
  swing1LimitDegrees=45
  swing1Motion=Free
  swing2LimitDegrees=45
  swing2Motion=Free
  twistLimitDegrees=45
  twistMotion=Free
 }
 {
  bone1Name=upperarm_r
  bone2Name=spine_04
  swing1LimitDegrees=45
  swing1Motion=Free
  swing2LimitDegrees=45
  swing2Motion=Free
  twistLimitDegrees=45
  twistMotion=Free
 }
 {
  bone1Name=def_boa_right_03
  bone2Name=spine_04
  swing1LimitDegrees=45
  swing1Motion=Free
  swing2LimitDegrees=45
  swing2Motion=Free
  twistLimitDegrees=45
  twistMotion=Free
 }
]
