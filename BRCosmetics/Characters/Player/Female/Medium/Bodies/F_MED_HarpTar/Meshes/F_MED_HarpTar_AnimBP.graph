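# Post-process-style cosmetic AnimBP for the F_MED_HarpTar female body part, adding secondary motion to the pose copied from its parent mesh.
# A Control Rig feeds crouch-switched RigidBody sims, then AnimDynamics chains for accessories, pendant and boa, finishing with bone offsets and a post-dynamics rig.
# Weapon changes briefly toggle standing physics and a boa offset.
# Updates set skydive/parachute cloth-wind material parameters, gravity and backbling state.
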AnimBlueprint /BRCosmetics/Characters/Player/Female/Medium/Bodies/F_MED_HarpTar/Meshes/F_MED_HarpTar_AnimBP
target_skeleton=&/F_MED_HarpTar_Skeleton
graphs [
 AnimGraph {nodes [
   n1 = "Transform (Modify) Bone - Bone: dyn_waist_bag"!(
    ComponentPose=Local To Component(LocalPose=Use cached pose 'Post dynamics'())
    bAlphaBoolEnabled:bool=Get bIsCrouching()
    Translation="0,1,0"
    Rotation="5,0,15"
    Scale="1,1,1"
   ) -> (Pose) {
    alphaInputType=Bool
    rotationMode=BMM_Additive
    rotationSpace=BCS_BoneSpace
    translationMode=BMM_Additive
    translationSpace=BCS_BoneSpace
    alphaBoolBlend {blendOption=Linear}
    boneToModify {boneName=dyn_waist_bag}
    rotation {pitch=5, roll=15}
    translation {y=1}
   }
   n2 = "Transform (Modify) Bone - Bone: def_boa_left_04"!(
    ComponentPose=@n1
    bAlphaBoolEnabled:bool=NOT Boolean(A=Get Changed weapon() -> (Changed weapon:bool))
    Translation="0,0,0"
    Rotation="15,0,0"
    Scale="1,1,1"
   ) -> (Pose) {
    alphaInputType=Bool
    rotationMode=BMM_Additive
    rotationSpace=BCS_BoneSpace
    translationMode=BMM_Additive
    translationSpace=BCS_BoneSpace
    alphaBoolBlend {blendOption=QuadraticInOut, blendOutTime=1}
    boneToModify {boneName=def_boa_left_04}
    rotation {pitch=15}
   }
   n3 = Control Rig!(
    Source=Component To Local(ComponentPose=@n2)
    Alpha
    Targeting:bool=Get bIsTargeting()
    Crouch:bool=Get bIsCrouching()
    PlayingEmote:bool=Get bIsPlayingEmote()
   ) -> (Pose) {
    alphaBoolBlend {blendOption=Linear}
    controlRigAssetReference {blueprintRigClass=&/F_MED_HarpTar_postDynamics_CtrlRig.F_MED_HarpTar_postDynamics_CtrlRig_C}
   }
   Output Pose(Result=@n3) {layerGroup=None}
   n4 = Control Rig!(
    Source=Copy Pose From Mesh() {bUseAttachedParent=true}
    Alpha
    RBAN:bool=Get bRigidBodyEnabled()
    crouching:bool=Get bIsCrouching()
    Parachute opened:bool=Get bIsParachuteOpened()
    Skidive floating:bool=Get bIsSkydiveFloating()
    playing emote:bool=Get bIsPlayingEmote()
    targeting:bool=Get bIsTargeting()
    jump or fall:bool=Get jump or fall() -> (jump or fall:bool)
    DBNO:bool=Get bIsDBNO()
   ) -> (Pose) {
    alphaBoolBlend {blendOption=Linear}
    controlRigAssetReference {blueprintRigClass=&/F_MED_HarpTar_CtrlRig.F_MED_HarpTar_CtrlRig_C}
   }
   Save cached pose 'PostControlRig'(Pose=@n4) {cacheName=PostControlRig}
   n5 = RigidBody!(ComponentPose=Local To Component(LocalPose=Use cached pose 'PostControlRig'()), Alpha, OverrideWorldGravity:Vector="0,0,-980", ExternalForce) -> (Pose) {
    bOverrideWorldGravity=true
    overridePhysicsAsset=&/F_MED_HarpTar_Crouch_Physics
    alphaBoolBlend {blendOption=Linear}
    overrideWorldGravity {z=-980}
    simSpaceSettings {dampingAlpha=0, maxAngularAcceleration=5, maxAngularVelocity=20, maxLinearAcceleration=3000, maxLinearVelocity=600, worldAlpha=1}
   }
   n6 = RigidBody!(
    ComponentPose=Local To Component(LocalPose=Use cached pose 'PostControlRig'())
    bAlphaBoolEnabled:bool=Get Changed weapon()
    OverrideWorldGravity="0,0,-980"
    ExternalForce
   ) -> (Pose) {
    alphaInputType=Bool
    bOverrideWorldGravity=true
    overridePhysicsAsset=&/F_MED_HarpTar_Physics
    simulationSpace=BaseBoneSpace
    alphaBoolBlend {blendOption=Linear}
    baseBoneRef {boneName=spine_05}
    componentAppliedLinearAccClamp=<1000 1000 1000>
    overrideWorldGravity {z=-980}
    simSpaceSettings {dampingAlpha=0, maxAngularAcceleration=20, maxAngularVelocity=5, maxLinearAcceleration=1200, maxLinearVelocity=500, worldAlpha=1}
   }
   Save cached pose 'Post RBAN'(Pose=Blend Poses by bool(BlendPose_0=Component To Local(ComponentPose=@n5), BlendPose_1=Component To Local(ComponentPose=@n6), bActiveValue=Get bIsCrouching()) {blendType=Linear}) {cacheName=Post RBAN}
   n7 = Anim Dynamics - Bone: dyn_pendant_03!(
    ComponentPose=Local To Component(LocalPose=Use cached pose 'FitsrPartDynamics'())
    Alpha=1
    GravityOverride:Vector="0,0,-980"
    ExternalForce="0,0,0"
   ) -> (Pose) {
    angularDampingOverride=0.999
    angularSpringConstant=100
    bAngularSpring=true
    bOverrideAngularDamping=true
    bUseGravityOverride=true
    bUsePlanarLimit=false
    bUseSphericalLimits=true
    gravityScale=0.35
    simulationSpace=RootRelative
    alphaBoolBlend {blendOption=Linear}
    boundBone {boneName=dyn_pendant_03}
    componentAppliedLinearAccClamp=<3000 3000 3000>
    componentLinearAccScale=<0.4 0.4 0.4>
    componentLinearVelScale=<0.4 0.4 0.4>
    gravityOverride {z=-980}
    physicsBodyDefinitions [{
      collisionType=CustomSphere
      sphereCollisionRadius=5.5
      boundBone {boneName=dyn_pendant_03}
      boxExtents=<15 15 15>
      constraintSetup {angularTargetAxis=AxisY, twistAxis=AxisY, angularLimitsMax=<0 0 8>, angularLimitsMin=<0 -5 -8>, angularTarget=<1 0 0>}
      localJointOffset=<1.5 0 0>
     }]
    simSpaceSettings {maxAngularAcceleration=35, maxAngularVelocity=7, simSpaceAngularAlpha=1}
    sphericalLimits [{limitRadius=4, drivingBone {boneName=spine_05}, sphereLocalOffset=<0 0 0>}]
   }
   n8 = Anim Dynamics - Bone: dyn_pendant_04!(ComponentPose=@n7, Alpha=1, GravityOverride:Vector="0,0,-980", ExternalForce="0,0,0") -> (Pose) {
    angularBiasOverride=0.999
    angularDampingOverride=0.999
    angularSpringConstant=100
    bAngularSpring=true
    bOverrideAngularDamping=true
    bUseGravityOverride=true
    bUseSphericalLimits=true
    gravityScale=0.35
    simulationSpace=RootRelative
    alphaBoolBlend {blendOption=Linear}
    boundBone {boneName=dyn_pendant_04}
    componentAppliedLinearAccClamp=<3000 3000 3000>
    componentLinearAccScale=<0.4 0.4 0.4>
    componentLinearVelScale=<0.4 0.4 0.4>
    gravityOverride {z=-980}
    physicsBodyDefinitions [{
      collisionType=CustomSphere
      sphereCollisionRadius=6
      boundBone {boneName=dyn_pendant_04}
      boxExtents=<15 15 15>
      constraintSetup {angularTargetAxis=AxisY, twistAxis=AxisY, angularLimitsMax=<0 0 8>, angularLimitsMin=<0 -5 -8>, angularTarget=<1 0 0>}
      localJointOffset=<1.5 0 0>
     }]
    simSpaceSettings {maxAngularAcceleration=35, maxAngularVelocity=7, simSpaceAngularAlpha=1}
    sphericalLimits [
     {limitRadius=5, drivingBone {boneName=spine_05}, sphereLocalOffset=<0 0 0>}
     {limitRadius=5, drivingBone {boneName=spine_04}, sphereLocalOffset=<0 2 0>}
    ]
   }
   n9 = Anim Dynamics - Bone: dyn_pendant_05!(ComponentPose=@n8, Alpha=1, GravityOverride:Vector="0,0,-980", ExternalForce="0,0,0") -> (Pose) {
    angularBiasOverride=0.999
    angularDampingOverride=0.999
    angularSpringConstant=100
    bAngularSpring=true
    bOverrideAngularDamping=true
    bUseGravityOverride=true
    bUseSphericalLimits=true
    gravityScale=0.35
    simulationSpace=RootRelative
    alphaBoolBlend {blendOption=Linear}
    boundBone {boneName=dyn_pendant_05}
    componentAppliedLinearAccClamp=<3000 3000 3000>
    componentLinearAccScale=<0.6 0.6 0.6>
    componentLinearVelScale=<0.6 0.6 0.6>
    gravityOverride {z=-980}
    physicsBodyDefinitions [{
      collisionType=CustomSphere
      sphereCollisionRadius=8
      boundBone {boneName=dyn_pendant_05}
      boxExtents=<15 15 15>
      constraintSetup {angularTargetAxis=AxisY, twistAxis=AxisY, angularLimitsMax=<0 0 8>, angularLimitsMin=<0 -5 -8>, angularTarget=<1 0 0>}
      localJointOffset=<1.5 0 0>
     }]
    simSpaceSettings {maxAngularAcceleration=35, maxAngularVelocity=7, simSpaceAngularAlpha=1}
    sphericalLimits [
     {limitRadius=5, drivingBone {boneName=spine_05}, sphereLocalOffset=<-5 0 0>}
     {limitRadius=5, drivingBone {boneName=spine_04}, sphereLocalOffset=<0 1 0>}
     {limitRadius=5, drivingBone {boneName=spine_04}, sphereLocalOffset=<0 1 -5>}
    ]
   }
   n10 = Component To Local(ComponentPose=@n9)
   n11 = Anim Dynamics - Bone: dyn_pendant_03(
    ComponentPose=Local To Component(LocalPose=Use cached pose 'FitsrPartDynamics'())
    Alpha=1
    GravityOverride="0,0,-980"
    ExternalForce="0,0,0"
   ) {
    angularDampingOverride=0.999
    angularSpringConstant=100
    bAngularSpring=true
    bOverrideAngularDamping=true
    bUseGravityOverride=true
    bUsePlanarLimit=false
    bUseSphericalLimits=true
    gravityScale=0.35
    simulationSpace=RootRelative
    alphaBoolBlend {blendOption=Linear}
    boundBone {boneName=dyn_pendant_03}
    componentAppliedLinearAccClamp=<3000 3000 3000>
    componentLinearAccScale=<0.4 0.4 0.4>
    componentLinearVelScale=<0.4 0.4 0.4>
    gravityOverride {z=-980}
    physicsBodyDefinitions [{
      collisionType=CustomSphere
      sphereCollisionRadius=5.5
      boundBone {boneName=dyn_pendant_03}
      boxExtents=<15 15 15>
      constraintSetup {angularTargetAxis=AxisY, twistAxis=AxisY, angularLimitsMax=<0 0 8>, angularLimitsMin=<0 -5 -8>, angularTarget=<1 0 0>}
      localJointOffset=<2 0 0>
     }]
    sphericalLimits [{limitRadius=4, drivingBone {boneName=spine_05}, sphereLocalOffset=<0 0 0>}]
   }
   n12 = Anim Dynamics - Bone: dyn_pendant_04(ComponentPose=@n11, Alpha=1, GravityOverride="0,0,-980", ExternalForce="0,0,0") {
    angularBiasOverride=0.999
    angularDampingOverride=0.999
    angularSpringConstant=100
    bAngularSpring=true
    bOverrideAngularDamping=true
    bUseGravityOverride=true
    bUseSphericalLimits=true
    gravityScale=0.35
    simulationSpace=RootRelative
    alphaBoolBlend {blendOption=Linear}
    boundBone {boneName=dyn_pendant_04}
    componentAppliedLinearAccClamp=<3000 3000 3000>
    componentLinearAccScale=<0.4 0.4 0.4>
    componentLinearVelScale=<0.4 0.4 0.4>
    gravityOverride {z=-980}
    physicsBodyDefinitions [{
      collisionType=CustomSphere
      sphereCollisionRadius=6
      boundBone {boneName=dyn_pendant_04}
      boxExtents=<15 15 15>
      constraintSetup {angularTargetAxis=AxisY, twistAxis=AxisY, angularLimitsMax=<0 0 8>, angularLimitsMin=<0 -5 -8>, angularTarget=<1 0 0>}
      localJointOffset=<2 0 0>
     }]
    sphericalLimits [
     {limitRadius=5, drivingBone {boneName=spine_05}, sphereLocalOffset=<0 0 0>}
     {limitRadius=5, drivingBone {boneName=spine_04}, sphereLocalOffset=<0 2 0>}
    ]
   }
   n13 = Anim Dynamics - Bone: dyn_pendant_05(ComponentPose=@n12, Alpha=1, GravityOverride="0,0,-980", ExternalForce="0,0,0") {
    angularBiasOverride=0.999
    angularDampingOverride=0.999
    angularSpringConstant=100
    bAngularSpring=true
    bOverrideAngularDamping=true
    bUseGravityOverride=true
    bUseSphericalLimits=true
    gravityScale=0.35
    simulationSpace=RootRelative
    alphaBoolBlend {blendOption=Linear}
    boundBone {boneName=dyn_pendant_05}
    componentAppliedLinearAccClamp=<3000 3000 3000>
    componentLinearAccScale=<0.6 0.6 0.6>
    componentLinearVelScale=<0.6 0.6 0.6>
    gravityOverride {z=-980}
    physicsBodyDefinitions [{
      collisionType=CustomSphere
      sphereCollisionRadius=8
      boundBone {boneName=dyn_pendant_05}
      boxExtents=<15 15 15>
      constraintSetup {angularTargetAxis=AxisY, twistAxis=AxisY, angularLimitsMax=<0 0 8>, angularLimitsMin=<0 -5 -8>, angularTarget=<1 0 0>}
      localJointOffset=<2 0 0>
     }]
    sphericalLimits [
     {limitRadius=5, drivingBone {boneName=spine_05}, sphereLocalOffset=<-5 0 0>}
     {limitRadius=5, drivingBone {boneName=spine_04}, sphereLocalOffset=<0 1 0>}
     {limitRadius=5, drivingBone {boneName=spine_04}, sphereLocalOffset=<0 1 -5>}
    ]
   }
   n14 = Component To Local(ComponentPose=@n13)
   n15 = Get bIsInFrontEnd()
   n16 = Anim Dynamics - Bone: dyn_back_boa!(
    ComponentPose=Local To Component(LocalPose=Blend Poses by bool(BlendPose_0=@n10, BlendPose_1=@n14, bActiveValue=@n15) {blendType=Linear})
    bAlphaBoolEnabled:bool=True
    GravityOverride:Vector="0,0,-980"
    ExternalForce="0,0,0"
   ) -> (Pose) {
    alphaInputType=Bool
    angularDampingOverride=0.999
    angularSpringConstant=220
    bAngularSpring=true
    bOverrideAngularDamping=true
    bUseGravityOverride=true
    gravityScale=0.35
    simulationSpace=RootRelative
    alphaBoolBlend {blendOption=Linear}
    boundBone {boneName=dyn_back_boa}
    componentAppliedLinearAccClamp=<3000 3000 3000>
    componentLinearAccScale=<0.3 0.4 0.4>
    componentLinearVelScale=<0.3 0.4 0.4>
    gravityOverride {z=-980}
    physicsBodyDefinitions [{
      boundBone {boneName=dyn_back_boa}
      boxExtents=<15 15 25>
      constraintSetup {angularLimitsMax=<0 20 5>, angularLimitsMin=<0 -20 -60>, angularTarget=<1 0 0>}
      localJointOffset=<3 0 0>
     }]
    relativeSpaceBone {boneName=spine_05}
   }
   Save cached pose 'Post dynamics'(Pose=Component To Local(ComponentPose=@n16)) {cacheName=Post dynamics}
   n17 = Anim Dynamics - Bone: dyn_braclet_02(ComponentPose=Local To Component(LocalPose=Use cached pose 'Post RBAN'()), Alpha=1, ExternalForce="0,0,0") {
    angularDampingOverride=1
    angularSpringConstant=20000
    bAngularSpring=true
    bOverrideAngularDamping=true
    bUseGravityOverride=true
    gravityScale=0.35
    simulationSpace=RootRelative
    alphaBoolBlend {blendOption=Linear}
    boundBone {boneName=dyn_braclet_02}
    componentLinearAccScale=<1 1 1>
    componentLinearVelScale=<1 1 1>
    gravityOverride {z=-980}
    physicsBodyDefinitions [{
      boundBone {boneName=dyn_braclet_02}
      boxExtents=<20 30 30>
      constraintSetup {angularLimitsMax=<15 15 15>, angularLimitsMin=<-15 -15 -15>, angularTarget=<1 0 0>}
      localJointOffset=<4 0 0>
     }]
   }
   n18 = Anim Dynamics - Bone: dyn_zip(ComponentPose=@n17, Alpha=1, ExternalForce="0,0,0") {
    angularDampingOverride=1
    angularSpringConstant=20000
    bAngularSpring=true
    bOverrideAngularDamping=true
    bUseGravityOverride=true
    gravityScale=0.35
    simulationSpace=RootRelative
    alphaBoolBlend {blendOption=Linear}
    boundBone {boneName=dyn_zip}
    componentLinearAccScale=<1 1 1>
    componentLinearVelScale=<1 1 1>
    gravityOverride {z=-980}
    physicsBodyDefinitions [{
      boundBone {boneName=dyn_zip}
      boxExtents=<15 15 15>
      constraintSetup {angularLimitsMax=<0 30 2>, angularLimitsMin=<0 0 -2>, angularTarget=<1 0 0>}
      localJointOffset=<4 0 0>
     }]
   }
   n19 = Anim Dynamics - Bone: dyn_waist_bag(ComponentPose=@n18, Alpha=1, ExternalForce="0,0,0") {
    angularDampingOverride=0.99
    angularSpringConstant=1000
    bAngularSpring=true
    bOverrideAngularDamping=true
    bUseGravityOverride=true
    gravityScale=0.35
    simulationSpace=RootRelative
    alphaBoolBlend {blendOption=Linear}
    boundBone {boneName=dyn_waist_bag}
    componentAppliedLinearAccClamp=<3000 3000 3000>
    componentLinearAccScale=<1 1 1>
    componentLinearVelScale=<1 1 1>
    gravityOverride {z=-980}
    physicsBodyDefinitions [{
      boundBone {boneName=dyn_waist_bag}
      boxExtents=<20 30 30>
      constraintSetup {angularLimitsMax=<0 5 8>, angularLimitsMin=<0 -7 -8>, angularTarget=<1 0 0>}
      localJointOffset=<4 0 0>
     }]
   }
   Save cached pose 'FitsrPartDynamics'(Pose=Component To Local(ComponentPose=@n19)) {cacheName=FitsrPartDynamics}
  ]}
 EventGraph {nodes [
   Event BlueprintUpdateAnimation()
   > n1 = Sequence!(execute) -> (then_0, then_1, then_2:Exec, then_3:Exec)
   > n2 = Gravity Override State Function DEPRECATED(CharacterPartAnimInstance=Self-Reference(), GameTimeInSeconds=GetGameTimeInSeconds(), GravityOverrideParams=Get Gravity Override Params())
   > Set Gravity Override(Gravity Override=@n2)
   > n3 = Set isWPOEnabled(isWPOEnabled:bool=OR Boolean!(A=Get bIsSkydiveFloating(), B=Get bIsSkydiveDiveMode(), C:bool=Get bIsParachuteOpened()) -> (ReturnValue)) -> (Output_Get:bool)
   > n4 = Branch(Condition=@n3)
   > n5 = Set WPO Pho Wind Intensity(WPO Pho Wind Intensity=1)
   SetScalarParameterValue(
    execute=[@n9.then, @n5.then]
    self=[
     Get Overlay Material Instance Dynamic() -> (Overlay Material Instance Dynamic:Material Instance Dynamic Object Reference)
     Get Material Instance Dynamic() -> (Material Instance Dynamic:Material Instance Dynamic Object Reference)
    ]
    ParameterName=ClothSim_WindStrength
    Value=Get WPO Pho Wind Intensity()
   )
   n6 = Get bIsParachuteOpened()
   n7 = Get bIsJumping()
   n8 = Get bIsFalling()
   > Set jump or fall(jump or fall:bool=AND Boolean(A=OR Boolean!(A=@n7, B=@n8) -> (ReturnValue), B=NOT Boolean(A=@n6))) -> (Output_Get:bool)
   n9 = Set WPO Pho Wind Intensity(execute=@n4.else, WPO Pho Wind Intensity=0)
   n10 = Branch(execute=@n1.then_1, Condition=Get bIsSkyDiving())
   > Set Velocity Interp Speed(Velocity Interp Speed:double=15) -> (Output_Get:double)
   > Set TurbulenceNoiseTileDivisor(TurbulenceNoiseTileDivisor:float=15) -> (Output_Get:float)
   > Set wind_strength(wind_strength:float=0.9) -> (Output_Get:float)
   Set Velocity Interp Speed(execute=@n10.else, Velocity Interp Speed=0)
   > Set TurbulenceNoiseTileDivisor(TurbulenceNoiseTileDivisor=0)
   > Set wind_strength(wind_strength=0)
   n11 = Set Skeletal Mesh(execute=@n1.then_2, Skeletal Mesh:Skeletal Mesh Component Object Reference=GetPartSkeletalMeshComponent()) -> (Output_Get:Skeletal Mesh Component Object Reference)
   > n12 = Get Dynamic Material Instance(SkeletalMeshComponent=@n11)
   > Set Material Instance Dynamic(Material Instance Dynamic:Material Instance Dynamic Object Reference=@n12) -> (Output_Get:Material Instance Dynamic Object Reference)
   > n13 = GetOrCreateDynamicOverlayMaterial(MeshComponent=@n11, MaterialName=MI_TechnoAwaken_Body)
   > Set Overlay Material Instance Dynamic(Overlay Material Instance Dynamic:Material Instance Dynamic Object Reference=@n13) -> (Output_Get:Material Instance Dynamic Object Reference)
   n14 = CheckForBackbling()
   Set withoutBackpack(execute=@n1.then_3, withoutBackpack:bool=NOT Boolean(A=@n14)) -> (Output_Get:bool)
   > n15 = Set isWearingPack(isWearingPack:bool=OR Boolean!(A=@n14, B=@n14.bIsBackblingACape, C:bool=@n14.bIsBackblingLong) -> (ReturnValue)) -> (Output_Get:bool)
   > Set isWearingCape(isWearingCape=@n14.bIsBackblingLong)
   > Set toggleAnimDynamics(toggleAnimDynamics:bool=NOT Boolean(A=@n15)) -> (Output_Get:bool)
   Event OnWeaponChanged()
   > Set Changed weapon(Changed weapon:bool=false) -> (Output_Get:bool)
   > Delay(Duration=0.18)
   > Set Changed weapon(Changed weapon=true)
  ]}
]
variables [
 Changed weapon {type=boolean}
 Gravity Override {type=Vector}
 Gravity Override Params {type=GravityOverrideParamsStruct}
 isWearingCape {type=boolean}
 isWearingPack {type=boolean}
 isWPOEnabled {type=boolean}
 jump or fall {type=boolean}
 Local Noise Offset {type=Vector}
 "Local Noise Transform  Location" {type=Vector}
 Mat Instance Dynamic {type=MaterialInstanceDynamic}
 Material Instance Dynamic {type=MaterialInstanceDynamic}
 Overlay Mat Instance Dynamic {type=MaterialInstanceDynamic}
 Overlay Material Instance Dynamic {type=MaterialInstanceDynamic}
 R1 Velocity Transform Location {type=Vector}
 R1 Velocity Transform Velocity {type=Vector}
 Skeletal Mesh {type=SkeletalMeshComponent}
 SkelMesh {type=SkeletalMeshComponent}
 toggleAnimDynamics {type=boolean}
 TurbulenceNoiseTileDivisor {type=number}
 Velocity Interp Speed {type=number}
 wind_strength {type=number}
 withoutBackpack {type=boolean}
 WPO Pho Wind Intensity {type=number}
]
