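# Control Rig for the HarpTar female medium face accessory that offsets the center, right and left hair master bones by body pose.
# A four-way sequence first applies relative transforms weighted by head spherical pose readers, then a skydive-floating offset smoothed through AlphaInterp.
# Crouch and FrontEnd flags gate further head-driven corrections.
# The spine_04 reader's output is unused, and one center-hair transform is unconnected.
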ControlRigBlueprint /BRCosmetics/Characters/Player/Female/Medium/Bodies/F_MED_HarpTar/Meshes/Parts/F_MED_HarpTar_FaceAcc_CtrlRig
graphs [
 RigVMFunctionLibrary {links [], nodes []}
 RigVMFunction_Sequence_1_ContainedGraph {nodes [
   n1 = RigVMFunctionEntryNode()
   RigVMFunction_Sequence = &RigVMFunction_Sequence(ExecuteContext=@n1.ExecuteContext)
   n2 = &RigVMFunction_Sequence(ExecuteContext=@RigVMFunction_Sequence.B)
   n3 = &RigVMFunction_Sequence(ExecuteContext=@n2.B)
   RigVMFunctionReturnNode(A=@RigVMFunction_Sequence.A, B=@n2.A, C=@n3.A, D=@n3.B)
  ]}
 RigVMModel {nodes [
   RigUnit_BeginExecution = &RigUnit_BeginExecution()
   RigVMFunction_Sequence = RigVMAggregateNode(ExecuteContext=@RigUnit_BeginExecution.ExecutePin)
   n1 = &RigUnit_SphericalPoseReader(
    ActiveRegionScaleFactors=(PositiveWidth=1.2,NegativeWidth=1.2,PositiveHeight=1,NegativeHeight=1)
    ActiveRegionSize=0.09
    Debug=(bDrawDebug=false,bDraw2D=False,bDrawLocalAxes=False,DebugScale=25,DebugSegments=20,DebugThickness=0.25)
    DriverAxis=(X=1,Y=1.3,Z=0)
    DriverItem=(Type=Bone,Name="head")
    FalloffRegionScaleFactors=(PositiveWidth=1.5,NegativeWidth=1.5,PositiveHeight=1,NegativeHeight=1)
    FalloffSize=0.2
    FlipHeightScaling=False
    FlipWidthScaling=False
    OptionalParentItem=(Type=Bone,Name="None")
    RotationOffset=(X=12,Y=-90,Z=0)
    ExecutePin=@RigVMFunction_Sequence.A
   )
   > RigUnit_SetRelativeTransformForItem = /Script/ControlRig.RigUnit_SetRelativeTransformForItem(
    Child=(Type=Bone,Name="MSTR_center_hair")
    Parent=(Type=Bone,Name="MSTR_center_hair")
    Value=(Rotation=(X=0,Y=-0.3827,Z=0,W=0.9239),Translation=(X=-5,Y=0,Z=-5),Scale3D=(X=1,Y=1,Z=1))
    bParentInitial=False
    bPropagateToChildren=True
    Weight=@n1.OutputParam
   ) {
    is=struct RigVMUnitNode
    template="Set Relative Transform::Execute(in Child,in Parent,in bParentInitial,in Value,in Weight,in bPropagateToChildren)"
   }
   > &RigUnit_SetRelativeTransformForItem(
    Child=(Type=Bone,Name="MSTR_right_hair")
    Parent=(Type=Bone,Name="MSTR_right_hair")
    Value=(Rotation=(X=-0.0076,Y=0.0868,Z=0.0868,W=0.9924),Translation=(X=-2,Y=0,Z=-1),Scale3D=(X=1,Y=1,Z=1))
    bParentInitial=False
    bPropagateToChildren=True
    Weight=@n1.OutputParam
   ) {template="Set Relative Transform::Execute(in Child,in Parent,in bParentInitial,in Value,in Weight,in bPropagateToChildren)"}
   > n2 = &RigUnit_SphericalPoseReader(
    ActiveRegionScaleFactors=(PositiveWidth=1,NegativeWidth=1,PositiveHeight=1,NegativeHeight=1)
    ActiveRegionSize=0.1
    Debug=(bDrawDebug=false,bDraw2D=False,bDrawLocalAxes=False,DebugScale=25,DebugSegments=20,DebugThickness=0.25)
    DriverAxis=(X=1,Y=1,Z=0)
    DriverItem=(Type=Bone,Name="head")
    FalloffRegionScaleFactors=(PositiveWidth=1,NegativeWidth=1,PositiveHeight=1,NegativeHeight=1)
    FalloffSize=0.2
    FlipHeightScaling=False
    FlipWidthScaling=False
    OptionalParentItem=(Type=Bone,Name="None")
    RotationOffset=(X=90,Y=0,Z=-15)
   )
   > &RigUnit_SetRelativeTransformForItem(
    Child=(Type=Bone,Name="MSTR_right_hair")
    Parent=(Type=Bone,Name="MSTR_right_hair")
    Value=(Rotation=(X=0,Y=-0.3007,Z=0,W=0.9537),Translation=(X=0,Y=0,Z=0),Scale3D=(X=1,Y=1,Z=1))
    bParentInitial=False
    bPropagateToChildren=True
    Weight=@n2.OutputParam
   ) {template="Set Relative Transform::Execute(in Child,in Parent,in bParentInitial,in Value,in Weight,in bPropagateToChildren)"}
   > &RigUnit_SphericalPoseReader(
    ActiveRegionScaleFactors=(PositiveWidth=1.5,NegativeWidth=1.5,PositiveHeight=1,NegativeHeight=1)
    ActiveRegionSize=0.1
    Debug=(bDrawDebug=false,bDraw2D=False,bDrawLocalAxes=False,DebugScale=25,DebugSegments=20,DebugThickness=0.25)
    DriverAxis=(X=0,Y=1,Z=0)
    DriverItem=(Type=Bone,Name="spine_04")
    FalloffRegionScaleFactors=(PositiveWidth=1,NegativeWidth=1,PositiveHeight=0.5,NegativeHeight=1)
    FalloffSize=0.24
    FlipHeightScaling=False
    FlipWidthScaling=False
    OptionalParentItem=(Type=Bone,Name="None")
    RotationOffset=(X=25,Y=-10,Z=0)
   )
   RigVMFunction_MathBoolToFloat = &RigVMFunction_MathBoolToFloat(Value=RigVMVariableNode() {variable=skidive floating}.Value)
   AlphaInterp = &RigVMFunction_AlphaInterp(InterpSpeedDecreasing=10, InterpSpeedIncreasing=5, Scale=1, bClampResult=False, Value=@RigVMFunction_MathBoolToFloat.Result)
   &RigUnit_SetRelativeTransformForItem(
    Child=(Type=Bone,Name="MSTR_center_hair")
    Parent=(Type=Bone,Name="MSTR_center_hair")
    Value=(Rotation=(X=-0.1305,Y=0,Z=0,W=0.9914),Translation=(X=-1,Y=0,Z=0),Scale3D=(X=1,Y=1,Z=1))
    bParentInitial=False
    bPropagateToChildren=True
    Weight=@AlphaInterp.Result
    ExecutePin=@RigVMFunction_Sequence.B
   ) {template="Set Relative Transform::Execute(in Child,in Parent,in bParentInitial,in Value,in Weight,in bPropagateToChildren)"}
   > &RigUnit_SetRelativeTransformForItem(
    Child=(Type=Bone,Name="MSTR_right_hair")
    Parent=(Type=Bone,Name="MSTR_right_hair")
    Value=(Rotation=(X=-0.173,Y=0.0858,Z=0.0151,W=0.9811),Translation=(X=0,Y=0,Z=0),Scale3D=(X=1,Y=1,Z=1))
    bParentInitial=False
    bPropagateToChildren=True
    Weight=@AlphaInterp.Result
   ) {template="Set Relative Transform::Execute(in Child,in Parent,in bParentInitial,in Value,in Weight,in bPropagateToChildren)"}
   > &RigUnit_SetRelativeTransformForItem(
    Child=(Type=Bone,Name="MSTR_left_hair")
    Parent=(Type=Bone,Name="MSTR_left_hair")
    Value=(Rotation=(X=0,Y=0,Z=0,W=1),Translation=(X=1,Y=0,Z=0),Scale3D=(X=1,Y=1,Z=1))
    bParentInitial=False
    bPropagateToChildren=True
    Weight=@AlphaInterp.Result
   ) {template="Set Relative Transform::Execute(in Child,in Parent,in bParentInitial,in Value,in Weight,in bPropagateToChildren)"}
   RigVMFunction_ControlFlowBranch = &RigVMFunction_ControlFlowBranch(ExecuteContext=@RigVMFunction_Sequence.C, Condition=RigVMVariableNode() {variable=Crouch}.Value)
   n3 = &RigUnit_SphericalPoseReader(
    ActiveRegionScaleFactors=(PositiveWidth=1,NegativeWidth=1,PositiveHeight=1,NegativeHeight=1)
    ActiveRegionSize=0.1
    Debug=(bDrawDebug=false,bDraw2D=False,bDrawLocalAxes=False,DebugScale=25,DebugSegments=20,DebugThickness=0.25)
    DriverAxis=(X=1,Y=0,Z=0)
    DriverItem=(Type=Bone,Name="head")
    FalloffRegionScaleFactors=(PositiveWidth=1,NegativeWidth=1,PositiveHeight=1,NegativeHeight=1)
    FalloffSize=0.2
    FlipHeightScaling=False
    FlipWidthScaling=False
    OptionalParentItem=(Type=Bone,Name="None")
    RotationOffset=(X=-55,Y=-65,Z=0)
    ExecutePin=@RigVMFunction_ControlFlowBranch.True
   )
   > &RigUnit_SetRelativeTransformForItem(
    Child=(Type=Bone,Name="MSTR_right_hair")
    Parent=(Type=Bone,Name="MSTR_right_hair")
    Value=(Rotation=(X=-0.0369,Y=0.5661,Z=-0.1083,W=0.8164),Translation=(X=-4,Y=0,Z=-1),Scale3D=(X=1,Y=1,Z=1))
    bParentInitial=False
    bPropagateToChildren=True
    Weight=@n3.OutputParam
   ) {template="Set Relative Transform::Execute(in Child,in Parent,in bParentInitial,in Value,in Weight,in bPropagateToChildren)"}
   > &RigUnit_SetRelativeTransformForItem(
    Child=(Type=Bone,Name="MSTR_center_hair")
    Parent=(Type=Bone,Name="MSTR_center_hair")
    Value=(Rotation=(X=-0.1163,Y=-0.7745,Z=0.0166,W=-0.6216),Translation=(X=-3,Y=0,Z=3),Scale3D=(X=1,Y=1,Z=1))
    bParentInitial=False
    bPropagateToChildren=True
    Weight=@n3.OutputParam
   ) {template="Set Relative Transform::Execute(in Child,in Parent,in bParentInitial,in Value,in Weight,in bPropagateToChildren)"}
   > &RigUnit_SetRelativeTransformForItem(
    Child=(Type=Bone,Name="MSTR_left_hair")
    Parent=(Type=Bone,Name="MSTR_left_hair")
    Value=(Rotation=(X=-0.2669,Y=-0.6076,Z=0.002,W=-0.7481),Translation=(X=-2,Y=0,Z=1),Scale3D=(X=1,Y=1,Z=1))
    bParentInitial=False
    bPropagateToChildren=True
    Weight=@n3.OutputParam
   ) {template="Set Relative Transform::Execute(in Child,in Parent,in bParentInitial,in Value,in Weight,in bPropagateToChildren)"}
   n4 = &RigVMFunction_ControlFlowBranch(ExecuteContext=@RigVMFunction_Sequence.D, Condition=RigVMVariableNode() {variable=FrontEnd}.Value)
   n5 = &RigUnit_SphericalPoseReader(
    ActiveRegionScaleFactors=(PositiveWidth=1,NegativeWidth=1,PositiveHeight=1,NegativeHeight=1)
    ActiveRegionSize=0.1
    Debug=(bDrawDebug=true,bDraw2D=False,bDrawLocalAxes=False,DebugScale=25,DebugSegments=20,DebugThickness=0.25)
    DriverAxis=(X=0,Y=1,Z=0)
    DriverItem=(Type=Bone,Name="head")
    FalloffRegionScaleFactors=(PositiveWidth=1,NegativeWidth=1,PositiveHeight=1,NegativeHeight=1)
    FalloffSize=0.2
    FlipHeightScaling=False
    FlipWidthScaling=False
    OptionalParentItem=(Type=Bone,Name="None")
    RotationOffset=(X=125,Y=20,Z=0)
    ExecutePin=@n4.True
   )
   > &RigUnit_SetRelativeTransformForItem(
    Child=(Type=Bone,Name="MSTR_left_hair")
    Parent=(Type=Bone,Name="MSTR_left_hair")
    Value=(Rotation=(X=0.0262,Y=0.0831,Z=-0.2996,W=0.9501),Translation=(X=0,Y=0,Z=0),Scale3D=(X=1,Y=1,Z=1))
    bParentInitial=False
    bPropagateToChildren=True
    Weight=@n5.OutputParam
   ) {template="Set Relative Transform::Execute(in Child,in Parent,in bParentInitial,in Value,in Weight,in bPropagateToChildren)"}
   &RigUnit_SetRelativeTransformForItem(
    Child=(Type=Bone,Name="MSTR_center_hair")
    ExecutePin=()
    Parent=(Type=Bone,Name="MSTR_center_hair")
    Value=(Rotation=(X=0,Y=0.766,Z=0,W=0.6428),Translation=(X=0,Y=0,Z=0),Scale3D=(X=1,Y=1,Z=1))
    Weight=1
    bParentInitial=False
    bPropagateToChildren=True
   ) {template="Set Relative Transform::Execute(in Child,in Parent,in bParentInitial,in Value,in Weight,in bPropagateToChildren)"}
  ]}
]
hierarchy [
 root {loc=<0 0 0>, rot=<0 0 0 1>}
 pelvis {parent=root, loc=<0 -0.5301 98.0571>, rot=<0.0127 0.707 -0.0127 -0.707>}
 spine_01 {parent=pelvis, loc=<6.1208 -0.9328 0>, rot=<0 0 0.034 -0.9994>}
 spine_02 {parent=spine_01, loc=<6.5997 -0.2095 0>, rot=<0 0 0.0238 -0.9997>}
 spine_03 {parent=spine_02, loc=<6.4914 -0.1237 0>, rot=<0 0 0.0149 -0.9999>}
 spine_04 {parent=spine_03, loc=<6.9598 0.0794 0>, rot=<0 0 -0.0132 -0.9999>}
 spine_05 {parent=spine_04, loc=<6.9987 0 0>, rot=<0 0 -0.0015 -1>}
 neck_01 {parent=spine_05, loc=<11.3138 0.1529 0>, rot=<0 0 -0.2003 -0.9797>}
 neck_02 {parent=neck_01, loc=<4.5627 0 0>, rot=<0 0 0.0504 -0.9987>}
 head {parent=neck_02, loc=<5.3914 0 0>, rot=<0 0 0.115 -0.9934>}
 MSTR_center_hair {parent=head, loc=<0.0349 -7.0463 -0.1541>, rot=<0.0092 -0.7098 -0.704 0.0235>}
 dyn_center_hair_01 {parent=MSTR_center_hair, loc=<0 0 0>, rot=<0 0 0 1>}
 dyn_center_hair_02 {parent=dyn_center_hair_01, loc=<5.385 0 0>, rot=<0 0 0 1>}
 MSTR_right_hair {parent=head, loc=<0.2398 -6.4003 4.5019>, rot=<-0.0255 -0.8702 -0.4921 -0.0032>}
 dyn_right_hair_01 {parent=MSTR_right_hair, loc=<0 0 0>, rot=<0 0 0 1>}
 dyn_right_hair_02 {parent=dyn_right_hair_01, loc=<4.224 0 0>, rot=<0 0 0 1>}
 MSTR_left_hair {parent=head, loc=<0.053 -6.4815 -4.825>, rot=<0.019 0.4278 0.9035 0.0173>}
 dyn_left_hair_01 {parent=MSTR_left_hair, loc=<0 0 0>, rot=<0 0 0 1>}
 dyn_left_hair_02 {parent=dyn_left_hair_01, loc=<4.2238 0 0>, rot=<0 0 0 1>}
 dyn_right_front_hair {parent=head, loc=<4.742 3.7732 7.5893>, rot=<0.0046 -0.0099 0.0237 -0.9997>}
 MSTR_bang {parent=head, loc=<10.1464 8.2952 -1.7546>, rot=<0.2462 0.6686 0.7015 0.0134>}
 dyn_bang_01 {parent=MSTR_bang, loc=<0 0 0>, rot=<0 0 0 1>}
 dyn_bang_02 {parent=dyn_bang_01, loc=<3.4813 0 0>, rot=<0 0 0 1>}
 MSTR_left_front_hair {parent=head, loc=<5.6205 3.4133 -7.2953>, rot=<-0.0754 -0.7075 -0.7 -0.062>}
 dyn_left_front_hair_01 {parent=MSTR_left_front_hair, loc=<0 0 0>, rot=<0 0 0 1>}
 dyn_left_front_hair_02 {parent=dyn_left_front_hair_01, loc=<3.7834 0 0>, rot=<0 0 0 1>}
 FACIAL_R_Jawline {parent=head, loc=<-2.9132 3.7089 3.8081>, rot=<0.6677 -0.4676 -0.4681 -0.3411>}
 FACIAL_L_Jawline {parent=head, loc=<-3.0708 3.6341 -4.1855>, rot=<-0.459 0.3563 0.6737 0.4567>}
 FACIAL_C_Jawline {parent=head, loc=<-5.9491 5.5475 -0.1441>, rot=<-0.6827 0.1638 0.6951 0.1546>}
 FACIAL_R_UnderChin {parent=head, loc=<-4.9438 2.3436 2.6227>, rot=<-0.3916 0.1097 0.8247 0.3932>}
 FACIAL_L_UnderChin {parent=head, loc=<-5.0507 2.2917 -2.8998>, rot=<0.8177 -0.3988 -0.4043 -0.0946>}
 FACIAL_C_UnderChin {parent=head, loc=<-6.1901 3.6273 -0.1226>, rot=<-0.6668 0.2232 0.6784 0.213>}
 dyn_bun_hair {parent=head, loc=<15.4363 -6.6038 -2.8967>, rot=<-0.0693 -0.8993 0.1078 0.4181>}
 dyn_left_locks_hair {parent=head, loc=<3.6901 -6.3657 -5.4569>, rot=<0.0039 -0.9519 -0.3065 0.0013>}
 dyn_bun_hair_01 {parent=head, loc=<12.0116 -6.1274 1.7767>, rot=<-0.0209 -0.989 -0.114 -0.0914>}
 MSTR_right_locks_hair {parent=head, loc=<2.8012 -5.9242 5.3802>, rot=<0.004 -0.9616 0.2745 -0.0011>}
 dyn_right_locks_hair {parent=MSTR_right_locks_hair, loc=<0 0 0>, rot=<0 0 0 1>}
 dyn_bun_hair_02 {parent=head, loc=<12.6383 -5.595 3.0622>, rot=<-0.3214 -0.9396 -0.0045 0.1176>}
]
variables [
 Crouch {type=bool, default=False}
 FrontEnd {type=bool, default=False}
 skidive floating {type=bool, default=False}
]
