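# Physics asset for the HarpTar medium female character's FaceAcc mesh, defining capsule bodies and constraints for dynamic hair.
# Five small MSTR anchors lock to the head, each leading a two-bone chain for bangs, center, left-front, left and right hair.
# Right-front hair and both side locks attach directly to the head, with narrow swing and zero-twist limits throughout the hair.
# Larger head and spine_05 capsules approximate the skull and upper torso, joined by a free-rotation constraint.
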PhysicsAsset /BRCosmetics/Characters/Player/Female/Medium/Bodies/F_MED_HarpTar/Meshes/Parts/F_MED_HarpTar_FaceAcc_Physics
bodies {
 MSTR_bang [{shapeType=Capsule, length=0.505, radius=0.505, shapeName=MSTR_bang_capsule, rotation=<rot 0 -90 0>}]
 MSTR_center_hair [{shapeType=Capsule, length=0.505, radius=0.505, shapeName=MSTR_center_hair_capsule, rotation=<rot 0 -90 0>}]
 MSTR_left_front_hair [{shapeType=Capsule, length=0.505, radius=0.505, shapeName=MSTR_left_front_hair_capsule, rotation=<rot 0 -90 0>}]
 MSTR_left_hair [{shapeType=Capsule, length=0.505, radius=0.505, shapeName=MSTR_left_hair_capsule, rotation=<rot 0 -90 0>}]
 MSTR_right_hair [{shapeType=Capsule, length=0.505, radius=0.505, shapeName=MSTR_right_hair_capsule, rotation=<rot 0 -90 0>}]
 dyn_bang_01 [{shapeType=Capsule, length=2.9743, radius=2.1472, shapeName=dyn_bang_01_capsule, center=<1.5364 0 0>, rotation=<rot 90 0 0>}]
 dyn_bang_02 [{
   shapeType=Capsule
   length=4.6605
   radius=3.0135
   shapeName=dyn_bang_02_capsule
   center=<2.1463 0.741 0.7072>
   rotation=<rot 64.8953 29.6453 0.0001>
  }]
 dyn_center_hair_01 [{
   shapeType=Capsule
   length=5.0593
   radius=3.6391
   shapeName=dyn_center_hair_01_capsule
   center=<2.3893 -0.3535 -1.3937>
   rotation=<rot -0.9121 90.2542 -120.3711>
  }]
 dyn_center_hair_02 [
  {
   shapeType=Capsule
   length=8.0229
   radius=2.5253
   shapeName=dyn_center_hair_02_capsule
   center=<5.3218 -1.1742 -3.9318>
   rotation=<rot 22.6269 -163.7394 -86.611>
  }
  {
   shapeType=Capsule
   length=2.6354
   radius=1.7282
   shapeName=dyn_center_hair_02_capsule
   center=<7.2159 -1.0643 -3.2702>
   rotation=<rot 72.1608 -16.4572 -54.8195>
  }
 ]
 dyn_left_front_hair_01 [{
   shapeType=Capsule
   length=4.1921
   radius=1.9306
   shapeName=dyn_left_front_hair_01_capsule
   center=<2.0572 0.162 0.4484>
   rotation=<rot 12.9857 -90 -13.7444>
  }]
 dyn_left_front_hair_02 [{
   shapeType=Capsule
   length=5.0026
   radius=2.5769
   shapeName=dyn_left_front_hair_02_capsule
   center=<3.357 -1.2535 0.5136>
   rotation=<rot 33.8534 21.8579 22.833>
  }]
 dyn_left_hair_01 [{
   shapeType=Capsule
   length=4.1772
   radius=2.6158
   shapeName=dyn_left_hair_01_capsule
   center=<-0.2282 0.6288 -1.9321>
   rotation=<rot 3.3726 -90 -83.1215>
  }]
 dyn_left_hair_02 [{
   shapeType=Capsule
   length=5.1161
   radius=3.0167
   shapeName=dyn_left_hair_02_capsule
   center=<3.3857 -0.5659 -1.2019>
   rotation=<rot 0 -33.3001 -50.2547>
  }]
 dyn_left_locks_hair [{
   shapeType=Capsule
   length=4.4868
   radius=3.3759
   shapeName=dyn_left_locks_hair_capsule
   center=<2.6208 -2.8852 0.8244>
   rotation=<rot 0 -107.2926 -81.0095>
  }]
 dyn_right_front_hair [{
   shapeType=Capsule
   length=4.2087
   radius=2.7762
   shapeName=dyn_right_front_hair_capsule
   center=<-3.0943 -0.0296 -1.0999>
   rotation=<rot 0 66.3178 -80.8088>
  }]
 dyn_right_hair_01 [{
   shapeType=Capsule
   length=5.3739
   radius=4.9003
   shapeName=dyn_right_hair_01_capsule
   center=<2.1433 1.2346 -0.3818>
   rotation=<rot 0 115.8856 -65.0001>
  }]
 dyn_right_hair_02 [
  {
   shapeType=Capsule
   length=3.8003
   radius=2.6844
   shapeName=dyn_right_hair_02_capsule
   center=<3.8283 -0.437 -3.0614>
   rotation=<rot 44.9504 -167.7376 -95.3942>
  }
  {
   shapeType=Capsule
   length=3.047
   radius=1.9418
   shapeName=dyn_right_hair_02_capsule
   center=<6.539 -2.9409 -1.9033>
   rotation=<rot -6.4519 27.8141 -77.9676>
  }
 ]
 dyn_right_locks_hair [{shapeType=Capsule, length=4.1863, radius=2.9917, shapeName=dyn_right_locks_hair_capsule, rotation=<rot 30.8374 -90 -71.3816>}]
 head [{shapeType=Capsule, length=8.8509, radius=7.53, center=<0.0222 0.4685 -0.0045>, rotation=<rot 90 0 0>}]
 spine_05 [
  {shapeType=Capsule, length=10.1989, radius=8.6254, shapeName=spine_05_capsule, center=<2.3889 0.6546 0>, rotation=<rot 0 -85.449 0>}
  {shapeType=Capsule, length=12.5381, radius=7.7351, shapeName=spine_05_capsule, center=<-3.8045 -16.2485 0>, rotation=<rot 0 3.5934 0>}
  {
   shapeType=Capsule
   length=11.9571
   radius=5.8738
   shapeName=spine_05_capsule
   center=<1.6897 -12.3843 11.6437>
   rotation=<rot -36.3573 -13.4487 28.204>
  }
 ]
}
constraints [
 {
  bone1Name=head
  bone2Name=spine_05
  swing1LimitDegrees=45
  swing1Motion=Free
  swing2LimitDegrees=45
  swing2Motion=Free
  twistLimitDegrees=45
  twistMotion=Free
 }
 {
  bone1Name=MSTR_right_hair
  bone2Name=head
  swing1LimitDegrees=45
  swing1Motion=Locked
  swing2LimitDegrees=45
  swing2Motion=Locked
  twistLimitDegrees=45
  twistMotion=Locked
 }
 {bone1Name=dyn_right_hair_01, bone2Name=MSTR_right_hair, swing1LimitDegrees=10, swing2LimitDegrees=15, twistLimitDegrees=0}
 {bone1Name=dyn_right_hair_02, bone2Name=dyn_right_hair_01, swing1LimitDegrees=10, swing2LimitDegrees=15, twistLimitDegrees=0}
 {
  bone1Name=MSTR_left_hair
  bone2Name=head
  swing1LimitDegrees=45
  swing1Motion=Locked
  swing2LimitDegrees=45
  swing2Motion=Locked
  twistLimitDegrees=45
  twistMotion=Locked
 }
 {bone1Name=dyn_left_hair_01, bone2Name=MSTR_left_hair, swing1LimitDegrees=15, swing2LimitDegrees=20, twistLimitDegrees=0}
 {bone1Name=dyn_left_hair_02, bone2Name=dyn_left_hair_01, swing1LimitDegrees=10, swing2LimitDegrees=15, twistLimitDegrees=0}
 {
  bone1Name=MSTR_center_hair
  bone2Name=head
  swing1LimitDegrees=45
  swing1Motion=Locked
  swing2LimitDegrees=45
  swing2Motion=Locked
  twistLimitDegrees=45
  twistMotion=Locked
 }
 {bone1Name=dyn_center_hair_01, bone2Name=MSTR_center_hair, swing1LimitDegrees=4.32, swing2LimitDegrees=30, twistLimitDegrees=0}
 {bone1Name=dyn_center_hair_02, bone2Name=dyn_center_hair_01, swing1LimitDegrees=4, swing2LimitDegrees=20, twistLimitDegrees=0}
 {bone1Name=dyn_right_front_hair, bone2Name=head, swing1LimitDegrees=15, swing2LimitDegrees=7, twistLimitDegrees=0}
 {
  bone1Name=MSTR_bang
  bone2Name=head
  swing1LimitDegrees=45
  swing1Motion=Locked
  swing2LimitDegrees=45
  swing2Motion=Locked
  twistLimitDegrees=45
  twistMotion=Locked
 }
 {bone1Name=dyn_bang_01, bone2Name=MSTR_bang, swing1LimitDegrees=15, swing2LimitDegrees=10, twistLimitDegrees=0}
 {bone1Name=dyn_bang_02, bone2Name=dyn_bang_01, swing1LimitDegrees=15, swing2LimitDegrees=8, twistLimitDegrees=0}
 {
  bone1Name=MSTR_left_front_hair
  bone2Name=head
  swing1LimitDegrees=45
  swing1Motion=Locked
  swing2LimitDegrees=45
  swing2Motion=Locked
  twistLimitDegrees=45
  twistMotion=Locked
 }
 {bone1Name=dyn_left_front_hair_01, bone2Name=MSTR_left_front_hair, swing1LimitDegrees=10, swing2LimitDegrees=20, twistLimitDegrees=0}
 {bone1Name=dyn_left_front_hair_02, bone2Name=dyn_left_front_hair_01, swing1LimitDegrees=10, swing2LimitDegrees=15, twistLimitDegrees=0}
 {bone1Name=dyn_left_locks_hair, bone2Name=head, swing1LimitDegrees=12, swing2LimitDegrees=12, twistLimitDegrees=0}
 {bone1Name=dyn_right_locks_hair, bone2Name=head, swing1LimitDegrees=15, swing2LimitDegrees=0, twistLimitDegrees=0}
]
